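# Editor utility widget that provides a labeled knob for editing a named float MetaSound input in weapon audio tooling.
# It looks up the input viewmodel by Input Name, and the WBP_Knob_Weapons child relays value changes through a dispatcher.
# Clicking the value text focuses a spinbox that commits to the float literal viewmodel's DisplayValue, while a revert button restores Default Value.
# Field notifications broadcast hover and editing state.
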EditorUtilityWidgetBlueprint /FortMetaSoundWeapon/Items/Metasounds/Widgets/EUW_Knob_Weapon_Default
graphs [
 EventGraph {nodes [
   "On Clicked (Button)"() {is=title K2Node_ComponentBoundEvent}
   > Set DisplayValue!(
    execute
    DisplayValue:float=Get Default Value() -> (Default Value:double)
    self:Meta Sound Literal Float Viewmodel Object Reference=Get MetaSoundLiteralViewModel_Float() -> (MetaSoundLiteralViewModel_Float:Meta Sound Literal Float Viewmodel Object Reference)
   ) -> (then, Output_Get:float)
   "On Clicked (ValueText_Button)"() {is=title K2Node_ComponentBoundEvent}
   > Set with Broadcast Is Editing Value!(execute, Is Editing Value:bool=true) -> (then, Output_Get:bool)
   > SetKeyboardFocus(self:Widget Object Reference=Get Value_Spinbox() -> (Value_Spinbox:Editor Utility Spin Box Object Reference))
   > SetForegroundColor!(
    execute
    self:Spin Box Object Reference=Get Value_Spinbox()
    InForegroundColor:Slate Color Structure=Make Slate Color(SpecifiedColor=(R=1,G=1,B=1,A=1))
   ) -> (then)
   n1 = "On Value Committed (Value_Spinbox)"!() -> (OutputDelegate, then, InValue:float, CommitMethod:ETextCommit Enum=Default) {is=title K2Node_ComponentBoundEvent}
   > Set DisplayValue(DisplayValue=@n1.InValue, self=Get MetaSoundLiteralViewModel_Float())
   > Set with Broadcast Is Editing Value(Is Editing Value=false)
   n2 = "On Value Changed (WBP_Knob_Weapons)"!() -> (OutputDelegate, then, Value:double) {is=title K2Node_ComponentBoundEvent}
   > Call On Paremeter Value Changed!(execute:Exec, self:EUW Knob Weapon Default Object Reference, Value:double=@n2.Value) -> (then:Exec) {is=title K2Node_CallDelegate}
   Event PreConstruct()
   > "SetText (Text)"(self=Get LabelTextBlock() -> (LabelTextBlock:Text Object Reference), InText=Get Label Text() -> (Label Text:Text))
   n3 = Event OnMouseEnter()
   Broadcast Field Value Changed!(execute=[@n3.then, @n4.then], self:Widget Object Reference, FieldId:Field Notification Id Structure=(FieldName="Is Hovered")) -> (then)
   n4 = Event OnMouseLeave!() -> (OutputDelegate, then, MouseEvent:"Pointer Event Structure (by ref)") {is=title K2Node_Event}
   n5 = Event Set Input Viewmodels!() -> (OutputDelegate, then, InputViewModels:Map of Names to Meta Sound Input Viewmodel Object References) {is=title K2Node_Event}
   > n6 = Cast To MetaSoundInputEditorViewModel!(
    execute
    Object=Find(
     TargetMap:Map of Names to Meta Sound Input Viewmodel Object References=@n5.InputViewModels
     Key:"Name (by ref)"=Get Input Name() -> (Input Name:Name=None)
    ) -> (Value:Meta Sound Input Viewmodel Object Reference)
   ) -> (then, CastFailed, AsMetaSound Input Editor Viewmodel:Meta Sound Input Editor Viewmodel Object Reference) {is=title K2Node_DynamicCast}
   > SetMetaSoundInputEditorViewModel!(execute, self, Viewmodel:Meta Sound Input Editor Viewmodel Object Reference=@n6) -> (then)
  ]}
 Get Reset Button Visibility {nodes [
   n1 = Get Reset Button Visibility!() -> (then, Value:double=~) {is=title K2Node_FunctionEntry}
   n2 = Get Default Value()
   > Return Node!(
    execute
    ReturnValue:ESlateVisibility Enum=Select(Option 0=Visible, Option 1=Hidden, Index="Nearly Equal (Float)"(A=@n1, B=@n2, ErrorTolerance=0))
   )
  ]}
 GetInputViewModelNames {nodes [
   Get Input Viewmodel Names() {is=title K2Node_FunctionEntry}
   > Return Node(ReturnValue:Array of Names=Make Array("[0]":Name=Get Input Name() -> (Input Name=None)) -> (Array:Array of Names))
  ]}
 Is Hovered {nodes [
   Is Hovered() {is=title K2Node_FunctionEntry}
   > Return Node(ReturnValue:bool=IsHovered())
  ]}
 OnParemeterValueChanged {nodes [OnParemeterValueChanged!() -> (then, Value:double=~) {is=title K2Node_FunctionEntry}]}
]
variables [
 Apply Logarithmic Interpolation {type=boolean}
 Default Value {type=number}
 Display Format {
  type=CommonNumberFormattingOptions
  default {
   alwaysSign=false
   maximumFractionalDigits=2
   maximumIntegralDigits=324
   minimumFractionalDigits=2
   minimumIntegralDigits=1
   roundingMode=HalfFromZero
   useGrouping=true
  }
 }
 Input Name {type=string, default=None}
 Is Editing Value {type=boolean}
 Knob Style {
  type=AudioMaterialKnobStyle
  default {
   material=null
   desiredSize=<64 64>
   knobAccentColor=<rgba 0.06 0.06 0.06 1>
   knobBarColor=<rgba 0.067 0.067 0.067 1>
   knobBarFillMaxColor=<rgba 1 1 1 1>
   knobBarFillMidColor=<rgba 0.96 0.96 0.96 1>
   knobBarFillMinColor=<rgba 1 1 1 1>
   knobBarFillTintColor=<rgba 0.96 0.96 0.96 1>
   knobBarShadowColor=<rgba 0.067 0.067 0.067 1>
   knobEdgeFillColor=<rgba 0.0156 0.0156 0.0156 1>
   knobIndicatorDotColor=<rgba 1 0 0 1>
   knobMainColor=<rgba 0.1406 0.1406 0.1406 1>
   knobShadowColor=<rgba 0.06 0.06 0.06 1>
   knobSmoothBevelColor=<rgba 0.0417 0.0417 0.0417 1>
   textBoxStyle {
    backgroundColor {colorUseRule=UseColor_ColorTable, specifiedColor=<rgba 0 0 0 0>}
    backgroundImage {
     bIsDynamicallyLoaded=false
     drawAs=RoundedBox
     imageType=NoImage
     mirroring=NoMirror
     resourceName=None
     resourceObject=null
     tiling=NoTile
     imageSize=<56 28>
     margin=<ltrb 0 0 0 0>
     outlineSettings {
      bUseBrushTransparency=false
      roundingType=FixedRadius
      width=0
      color {colorUseRule=UseColor_Specified, specifiedColor=<rgba 0 0 0 0>}
      cornerRadii=<4 4 4 4>
     }
     tintColor {colorUseRule=UseColor_ColorTable, specifiedColor=<rgba 0 0 0 0>}
     uVRegion {bIsValid=false, max=<0 0>, min=<0 0>}
    }
   }
  }
 }
 Label Text {type=string, default=Label}
 OnParemeterValueChanged {default=()}
]
widgets {
 info {rootWidgetClass=&VerticalBox, widgetCount=11}
 widgets [
  VerticalBox_1 {widgetClassPath=&VerticalBox, widget=VerticalBox_1, properties {visibility=Visible}}
  LabelTextBlock {
   widgetClassPath=&TextBlock
   bIsVariable=true
   parent=VerticalBox_1
   widget=LabelTextBlock
   properties {
    justification=Center
    text=Label
    colorAndOpacity {colorUseRule=UseColor_Specified, specifiedColor=<rgba 0.5271 0.5271 0.5271 1>}
    font {
     bForceMonospaced=false
     bMaterialIsStencil=false
     fontMaterial=null
     fontObject=&Roboto
     letterSpacing=0
     monospacedWidth=1
     size=9
     skewAmount=0
     typefaceFontName=Regular
     outlineSettings {
      bApplyOutlineToDropShadows=false
      bMiteredCorners=false
      bSeparateFillAlpha=false
      outlineMaterial=null
      outlineSize=0
      outlineColor=<rgba 0 0 0 1>
     }
    }
   }
   slot {class=VerticalBoxSlot, properties {horizontalAlignment=HAlign_Center, padding=<ltrb 0 8 0 0>}}
  }
  Spacer {widgetClassPath=&Spacer, parent=VerticalBox_1, widget=Spacer, properties {size=<1 8>}, slot {class=VerticalBoxSlot}}
  WBP_Knob_Weapons {
   widgetClassPath=&/WBP_Knob_Weapons
   bIsVariable=true
   parent=VerticalBox_1
   widget=WBP_Knob_Weapons
   properties {
    metaSoundLiteralViewModel_Float=&self:WidgetTree.WBP_Knob_Weapons.MetaSoundLiteralViewModel_Float_0
    knob Style {
     material=null
     desiredSize=<96 96>
     knobAccentColor=<rgba 0.06 0.06 0.06 1>
     knobBarColor=<rgba 0.067 0.067 0.067 1>
     knobBarFillMaxColor=<rgba 0.8707 1 0 1>
     knobBarFillMidColor=<rgba 0.3169 1 0 1>
     knobBarFillMinColor=<rgba 0 1 0.0698 1>
     knobBarFillTintColor=<rgba 0.96 0.96 0.96 1>
     knobBarShadowColor=<rgba 0.067 0.067 0.067 1>
     knobEdgeFillColor=<rgba 0.0156 0.0156 0.0156 1>
     knobIndicatorDotColor=<rgba 0 1 0.0698 1>
     knobMainColor=<rgba 0.1406 0.1406 0.1406 1>
     knobShadowColor=<rgba 0.06 0.06 0.06 1>
     knobSmoothBevelColor=<rgba 0.0417 0.0417 0.0417 1>
     textBoxStyle {
      backgroundColor {colorUseRule=UseColor_ColorTable, specifiedColor=<rgba 0 0 0 0>}
      backgroundImage {
       bIsDynamicallyLoaded=false
       drawAs=RoundedBox
       imageType=NoImage
       mirroring=NoMirror
       resourceName=None
       resourceObject=null
       tiling=NoTile
       imageSize=<56 28>
       margin=<ltrb 0 0 0 0>
       outlineSettings {
        bUseBrushTransparency=false
        roundingType=FixedRadius
        width=0
        color {colorUseRule=UseColor_Specified, specifiedColor=<rgba 0 0 0 0>}
        cornerRadii=<4 4 4 4>
       }
       tintColor {colorUseRule=UseColor_ColorTable, specifiedColor=<rgba 0 0 0 0>}
       uVRegion {bIsValid=false, max=<0 0>, min=<0 0>}
      }
     }
    }
   }
   slot {class=VerticalBoxSlot}
  }
  Spacer_232 {widgetClassPath=&Spacer, parent=VerticalBox_1, widget=Spacer_232, properties {size=<1 8>}, slot {class=VerticalBoxSlot}}
  VerticalBox_6 {
   widgetClassPath=&VerticalBox
   parent=VerticalBox_1
   widget=VerticalBox_6
   properties {bOverride_Cursor=true}
   slot {class=VerticalBoxSlot, properties {verticalAlignment=VAlign_Bottom}}
  }
  ValueText_Button {
   widgetClassPath=&EditorUtilityButton
   parent=VerticalBox_6
   widget=ValueText_Button
   properties {
    bOverride_Cursor=true
    cursor=TextEditBeam
    widgetStyle {
     clickedSlateSound {resourceObject=null}
     disabled {
      bIsDynamicallyLoaded=false
      drawAs=NoDrawType
      imageType=NoImage
      mirroring=NoMirror
      resourceName=None
      resourceObject=null
      tiling=NoTile
      imageSize=<32 32>
      margin=<ltrb 0 0 0 0>
      outlineSettings {
       bUseBrushTransparency=false
       roundingType=FixedRadius
       width=1
       color {colorUseRule=UseColor_ColorTable, specifiedColor=<rgba 0 0 0 0>}
       cornerRadii=<4 4 4 4>
      }
      tintColor {colorUseRule=UseColor_Specified, specifiedColor=<rgba 0.0395 0.0395 0.0395 1>}
      uVRegion {bIsValid=false, max=<0 0>, min=<0 0>}
     }
     disabledForeground {colorUseRule=UseColor_ColorTable, specifiedColor=<rgba 0 0 0 0>}
     hovered {
      bIsDynamicallyLoaded=false
      drawAs=NoDrawType
      imageType=NoImage
      mirroring=NoMirror
      resourceName=None
      resourceObject=null
      tiling=NoTile
      imageSize=<32 32>
      margin=<ltrb 0 0 0 0>
      outlineSettings {
       bUseBrushTransparency=false
       roundingType=FixedRadius
       width=1
       color {colorUseRule=UseColor_ColorTable, specifiedColor=<rgba 0 0 0 0>}
       cornerRadii=<4 4 4 4>
      }
      tintColor {colorUseRule=UseColor_Specified, specifiedColor=<rgba 0.0395 0.0395 0.0395 1>}
      uVRegion {bIsValid=false, max=<0 0>, min=<0 0>}
     }
     hoveredForeground {colorUseRule=UseColor_ColorTable, specifiedColor=<rgba 0 0 0 0>}
     hoveredSlateSound {resourceObject=null}
     normal {
      bIsDynamicallyLoaded=false
      drawAs=NoDrawType
      imageType=NoImage
      mirroring=NoMirror
      resourceName=None
      resourceObject=null
      tiling=NoTile
      imageSize=<32 32>
      margin=<ltrb 0 0 0 0>
      outlineSettings {
       bUseBrushTransparency=false
       roundingType=FixedRadius
       width=1
       color {colorUseRule=UseColor_ColorTable, specifiedColor=<rgba 0 0 0 0>}
       cornerRadii=<4 4 4 4>
      }
      tintColor {colorUseRule=UseColor_Specified, specifiedColor=<rgba 0.0395 0.0395 0.0395 1>}
      uVRegion {bIsValid=false, max=<0 0>, min=<0 0>}
     }
     normalForeground {colorUseRule=UseColor_ColorTable, specifiedColor=<rgba 0 0 0 0>}
     normalPadding=<ltrb 0 0 0 0>
     pressed {
      bIsDynamicallyLoaded=false
      drawAs=NoDrawType
      imageType=NoImage
      mirroring=NoMirror
      resourceName=None
      resourceObject=null
      tiling=NoTile
      imageSize=<32 32>
      margin=<ltrb 0 0 0 0>
      outlineSettings {
       bUseBrushTransparency=false
       roundingType=FixedRadius
       width=1
       color {colorUseRule=UseColor_ColorTable, specifiedColor=<rgba 0 0 0 0>}
       cornerRadii=<4 4 4 4>
      }
      tintColor {colorUseRule=UseColor_Specified, specifiedColor=<rgba 0.0395 0.0395 0.0395 1>}
      uVRegion {bIsValid=false, max=<0 0>, min=<0 0>}
     }
     pressedForeground {colorUseRule=UseColor_ColorTable, specifiedColor=<rgba 0 0 0 0>}
     pressedPadding=<ltrb 0 0 0 0>
     pressedSlateSound {resourceObject=null}
    }
   }
   slot {class=VerticalBoxSlot, properties {horizontalAlignment=HAlign_Center}}
  }
  Value_TextBlock {
   widgetClassPath=&TextBlock
   bIsVariable=true
   parent=ValueText_Button
   widget=Value_TextBlock
   properties {
    justification=Center
    text=Value
    colorAndOpacity {colorUseRule=UseColor_Specified, specifiedColor=<rgba 0.5271 0.5271 0.5271 1>}
    font {
     bForceMonospaced=false
     bMaterialIsStencil=false
     fontMaterial=null
     fontObject=&Roboto
     letterSpacing=0
     monospacedWidth=1
     size=9
     skewAmount=0
     typefaceFontName=Regular
     outlineSettings {
      bApplyOutlineToDropShadows=false
      bMiteredCorners=false
      bSeparateFillAlpha=false
      outlineMaterial=null
      outlineSize=0
      outlineColor=<rgba 0 0 0 1>
     }
    }
   }
   slot {class=ButtonSlot, properties {horizontalAlignment=HAlign_Fill, verticalAlignment=VAlign_Fill, padding=<ltrb 0 0 0 0>}}
  }
  Value_Spinbox {
   widgetClassPath=/Script/Blutility.EditorUtilitySpinBox
   bIsVariable=true
   parent=VerticalBox_6
   widget=Value_Spinbox
   properties {
    bEnableSlider=false
    font {
     bForceMonospaced=false
     bMaterialIsStencil=false
     fontMaterial=null
     fontObject=&Roboto
     letterSpacing=0
     monospacedWidth=1
     size=9
     skewAmount=0
     typefaceFontName=Regular
     outlineSettings {
      bApplyOutlineToDropShadows=false
      bMiteredCorners=false
      bSeparateFillAlpha=false
      outlineMaterial=null
      outlineSize=0
      outlineColor=<rgba 0 0 0 1>
     }
    }
    foregroundColor {colorUseRule=UseColor_Specified, specifiedColor=<rgba 0.5271 0.5271 0.5271 1>}
    widgetStyle {
     activeBackgroundBrush {
      bIsDynamicallyLoaded=false
      drawAs=NoDrawType
      imageType=NoImage
      mirroring=NoMirror
      resourceName=None
      resourceObject=null
      tiling=NoTile
      imageSize=<0 0>
      margin=<ltrb 0 0 0 0>
      outlineSettings {
       bUseBrushTransparency=false
       roundingType=FixedRadius
       width=1
       color {colorUseRule=UseColor_ColorTable, specifiedColor=<rgba 0 0 0 0>}
       cornerRadii=<4 4 4 4>
      }
      tintColor {colorUseRule=UseColor_Specified, specifiedColor=<rgba 0.0048 0.0048 0.0048 1>}
      uVRegion {bIsValid=false, max=<0 0>, min=<0 0>}
     }
     activeFillBrush {
      bIsDynamicallyLoaded=false
      drawAs=NoDrawType
      imageType=NoImage
      mirroring=NoMirror
      resourceName=None
      resourceObject=null
      tiling=NoTile
      imageSize=<0 0>
      margin=<ltrb 0 0 0 0>
      outlineSettings {
       bUseBrushTransparency=false
       roundingType=FixedRadius
       width=1
       color {colorUseRule=UseColor_ColorTable, specifiedColor=<rgba 0 0 0 0>}
       cornerRadii=<4 4 4 4>
      }
      tintColor {colorUseRule=UseColor_Specified, specifiedColor=<rgba 0.0048 0.0048 0.0048 1>}
      uVRegion {bIsValid=false, max=<0 0>, min=<0 0>}
     }
     arrowsImage {
      bIsDynamicallyLoaded=false
      drawAs=NoDrawType
      imageType=NoImage
      mirroring=NoMirror
      resourceName=None
      resourceObject=null
      tiling=NoTile
      imageSize=<0 0>
      margin=<ltrb 0 0 0 0>
      outlineSettings {
       bUseBrushTransparency=false
       roundingType=HalfHeightRadius
       width=0
       color {colorUseRule=UseColor_Specified, specifiedColor=<rgba 0 0 0 0>}
       cornerRadii=<0 0 0 0>
      }
      tintColor {colorUseRule=UseColor_Specified, specifiedColor=<rgba 1 1 1 1>}
      uVRegion {bIsValid=false, max=<0 0>, min=<0 0>}
     }
     backgroundBrush {
      bIsDynamicallyLoaded=false
      drawAs=NoDrawType
      imageType=NoImage
      mirroring=NoMirror
      resourceName=None
      resourceObject=null
      tiling=NoTile
      imageSize=<0 0>
      margin=<ltrb 0 0 0 0>
      outlineSettings {
       bUseBrushTransparency=false
       roundingType=FixedRadius
       width=1
       color {colorUseRule=UseColor_ColorTable, specifiedColor=<rgba 0 0 0 0>}
       cornerRadii=<4 4 4 4>
      }
      tintColor {colorUseRule=UseColor_Specified, specifiedColor=<rgba 0.0048 0.0048 0.0048 1>}
      uVRegion {bIsValid=false, max=<0 0>, min=<0 0>}
     }
     foregroundColor {colorUseRule=UseColor_Specified, specifiedColor=<rgba 1 1 1 1>}
     hoveredBackgroundBrush {
      bIsDynamicallyLoaded=false
      drawAs=NoDrawType
      imageType=NoImage
      mirroring=NoMirror
      resourceName=None
      resourceObject=null
      tiling=NoTile
      imageSize=<0 0>
      margin=<ltrb 0 0 0 0>
      outlineSettings {
       bUseBrushTransparency=false
       roundingType=FixedRadius
       width=1
       color {colorUseRule=UseColor_ColorTable, specifiedColor=<rgba 0 0 0 0>}
       cornerRadii=<4 4 4 4>
      }
      tintColor {colorUseRule=UseColor_Specified, specifiedColor=<rgba 0.0048 0.0048 0.0048 1>}
      uVRegion {bIsValid=false, max=<0 0>, min=<0 0>}
     }
     hoveredFillBrush {
      bIsDynamicallyLoaded=false
      drawAs=NoDrawType
      imageType=NoImage
      mirroring=NoMirror
      resourceName=None
      resourceObject=null
      tiling=NoTile
      imageSize=<0 0>
      margin=<ltrb 0 0 0 0>
      outlineSettings {
       bUseBrushTransparency=false
       roundingType=FixedRadius
       width=1
       color {colorUseRule=UseColor_ColorTable, specifiedColor=<rgba 0 0 0 0>}
       cornerRadii=<4 4 4 4>
      }
      tintColor {colorUseRule=UseColor_Specified, specifiedColor=<rgba 0.0048 0.0048 0.0048 1>}
      uVRegion {bIsValid=false, max=<0 0>, min=<0 0>}
     }
     inactiveFillBrush {
      bIsDynamicallyLoaded=false
      drawAs=NoDrawType
      imageType=NoImage
      mirroring=NoMirror
      resourceName=None
      resourceObject=null
      tiling=NoTile
      imageSize=<0 0>
      margin=<ltrb 0 0 0 0>
      outlineSettings {
       bUseBrushTransparency=false
       roundingType=FixedRadius
       width=1
       color {colorUseRule=UseColor_ColorTable, specifiedColor=<rgba 0 0 0 0>}
       cornerRadii=<4 4 4 4>
      }
      tintColor {colorUseRule=UseColor_Specified, specifiedColor=<rgba 0.0048 0.0048 0.0048 1>}
      uVRegion {bIsValid=false, max=<0 0>, min=<0 0>}
     }
     insetPadding=<ltrb 0 0 0 0>
     textPadding=<ltrb 0 0 0 0>
    }
   }
   slot {class=VerticalBoxSlot, properties {horizontalAlignment=HAlign_Center}}
  }
  Button {
   widgetClassPath=&EditorUtilityButton
   parent=VerticalBox_6
   widget=Button
   properties {
    toolTipText=Reset to default value.
    widgetStyle {
     clickedSlateSound {resourceObject=null}
     disabled {
      bIsDynamicallyLoaded=false
      drawAs=NoDrawType
      imageType=NoImage
      mirroring=NoMirror
      resourceName=None
      resourceObject=null
      tiling=NoTile
      imageSize=<32 32>
      margin=<ltrb 0 0 0 0>
      outlineSettings {
       bUseBrushTransparency=false
       roundingType=FixedRadius
       width=1
       color {colorUseRule=UseColor_ColorTable, specifiedColor=<rgba 0 0 0 0>}
       cornerRadii=<4 4 4 4>
      }
      tintColor {colorUseRule=UseColor_Specified, specifiedColor=<rgba 0.0312 0.0312 0.0312 1>}
      uVRegion {bIsValid=false, max=<0 0>, min=<0 0>}
     }
     disabledForeground {colorUseRule=UseColor_ColorTable, specifiedColor=<rgba 0 0 0 0>}
     hovered {
      bIsDynamicallyLoaded=false
      drawAs=RoundedBox
      imageType=NoImage
      mirroring=NoMirror
      resourceName=None
      resourceObject=null
      tiling=NoTile
      imageSize=<32 32>
      margin=<ltrb 0 0 0 0>
      outlineSettings {
       bUseBrushTransparency=false
       roundingType=FixedRadius
       width=0
       color {colorUseRule=UseColor_Foreground, specifiedColor=<rgba 0 0 0 0>}
       cornerRadii=<4 4 4 4>
      }
      tintColor {colorUseRule=UseColor_Specified, specifiedColor=<rgba 0.0395 0.0395 0.0395 1>}
      uVRegion {bIsValid=false, max=<0 0>, min=<0 0>}
     }
     hoveredForeground {colorUseRule=UseColor_ColorTable, specifiedColor=<rgba 0 0 0 0>}
     hoveredSlateSound {resourceObject=null}
     normal {
      bIsDynamicallyLoaded=false
      drawAs=NoDrawType
      imageType=NoImage
      mirroring=NoMirror
      resourceName=None
      resourceObject=null
      tiling=NoTile
      imageSize=<32 32>
      margin=<ltrb 0 0 0 0>
      outlineSettings {
       bUseBrushTransparency=false
       roundingType=FixedRadius
       width=1
       color {colorUseRule=UseColor_ColorTable, specifiedColor=<rgba 0 0 0 0>}
       cornerRadii=<4 4 4 4>
      }
      tintColor {colorUseRule=UseColor_Specified, specifiedColor=<rgba 0.0312 0.0312 0.0312 1>}
      uVRegion {bIsValid=false, max=<0 0>, min=<0 0>}
     }
     normalForeground {colorUseRule=UseColor_ColorTable, specifiedColor=<rgba 0 0 0 0>}
     normalPadding=<ltrb 2 0 2 0>
     pressed {
      bIsDynamicallyLoaded=false
      drawAs=RoundedBox
      imageType=NoImage
      mirroring=NoMirror
      resourceName=None
      resourceObject=null
      tiling=NoTile
      imageSize=<32 32>
      margin=<ltrb 0 0 0 0>
      outlineSettings {
       bUseBrushTransparency=false
       roundingType=FixedRadius
       width=0
       color {colorUseRule=UseColor_Foreground, specifiedColor=<rgba 0 0 0 0>}
       cornerRadii=<4 4 4 4>
      }
      tintColor {colorUseRule=UseColor_Specified, specifiedColor=<rgba 0.0395 0.0395 0.0395 1>}
      uVRegion {bIsValid=false, max=<0 0>, min=<0 0>}
     }
     pressedForeground {colorUseRule=UseColor_ColorTable, specifiedColor=<rgba 0 0 0 0>}
     pressedPadding=<ltrb 2 0.5 2 -0.5>
     pressedSlateSound {resourceObject=null}
    }
   }
   slot {class=VerticalBoxSlot, properties {horizontalAlignment=HAlign_Center, verticalAlignment=VAlign_Center}}
  }
  Icon {
   widgetClassPath=/Script/TechAudioToolsContentEditor.TechAudioToolsEditorSlateIcon
   bIsVariable=true
   parent=Button
   widget=Icon
   properties {fullIconPath=DefaultRevisionControlStyle.RevisionControl.Actions.Revert, imageColorAndOpacity=<rgba 0.2135 0.2135 0.2135 1>, size=<12 12>}
   slot {class=ButtonSlot, properties {padding=<ltrb 2 2 2 2>}}
  }
 ]
}
