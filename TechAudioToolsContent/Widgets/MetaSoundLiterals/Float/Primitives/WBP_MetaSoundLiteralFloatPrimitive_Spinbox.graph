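# Float spinbox primitive widget for MetaSound literals, wrapping a single EditorUtilitySpinBox.
# It reports its Input Name as the required viewmodel name, then picks that entry from the supplied map and sets it as its MetaSound input viewmodel.
# Slider begin and end events are re-broadcast through the widget's own dispatchers.
# Mouse enter and leave broadcast a field notification for the Is Hovered function.
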WidgetBlueprint /TechAudioToolsContent/Widgets/MetaSoundLiterals/Float/Primitives/WBP_MetaSoundLiteralFloatPrimitive_Spinbox
graphs [
 EventGraph {nodes [
   "On Begin Slider Movement (Spinbox)"() {is=title K2Node_ComponentBoundEvent}
   > Call On Begin Slider Movement!(execute:Exec, self:Meta Sound Literal Float Primitive Spinbox Object Reference) -> (then:Exec) {is=title K2Node_CallDelegate}
   "On End Slider Movement (Spinbox)"!() -> (OutputDelegate, then, InValue:float) {is=title K2Node_ComponentBoundEvent}
   > Call On End Slider Movement!(execute:Exec, self:Meta Sound Literal Float Primitive Spinbox Object Reference) -> (then:Exec) {is=title K2Node_CallDelegate}
   n1 = Event OnMouseEnter()
   Broadcast Field Value Changed!(execute=[@n1.then, @n2.then], self:Widget Object Reference, FieldId:Field Notification Id Structure=(FieldName="Is Hovered")) -> (then)
   n2 = Event OnMouseLeave!() -> (OutputDelegate, then, MouseEvent:"Pointer Event Structure (by ref)") {is=title K2Node_Event}
   n3 = Event Set Input Viewmodels!() -> (OutputDelegate, then, InputViewModels:Map of Names to Meta Sound Input Viewmodel Object References) {is=title K2Node_Event}
   > SetMetaSoundInputViewModel!(
    execute
    self
    Viewmodel:Meta Sound Input Viewmodel Object Reference=Find(
     TargetMap:Map of Names to Meta Sound Input Viewmodel Object References=@n3.InputViewModels
     Key:"Name (by ref)"=Get Input Name() -> (Input Name:Name=None)
    ) -> (Value:Meta Sound Input Viewmodel Object Reference)
   ) -> (then)
  ]}
 GetInputViewModelNames {nodes [
   Get Input Viewmodel Names() {is=title K2Node_FunctionEntry}
   > Return Node!(execute, ReturnValue:Array of Names=Make Array("[0]":Name=Get Input Name() -> (Input Name=None)) -> (Array:Array of Names))
  ]}
 Is Hovered {nodes [
   Is Hovered() {is=title K2Node_FunctionEntry}
   > Return Node(ReturnValue:bool=IsHovered())
  ]}
 On Begin Slider Movement {nodes [On Begin Slider Movement() {is=title K2Node_FunctionEntry}]}
 On End Slider Movement {nodes [On End Slider Movement() {is=title K2Node_FunctionEntry}]}
]
variables [
 Input Name {type=string, default=None}
 On Begin Slider Movement {}
 On End Slider Movement {}
]
widgets {
 info {inheritedWidgetCount=0, parentClass=&UserWidget, rootWidgetClass=/Script/Blutility.EditorUtilitySpinBox, widgetCount=1}
 widgets [Spinbox {
   widgetClassPath=&EditorUtilitySpinBox
   bIsVariable=true
   widget=Spinbox
   properties {
    justification=Center
    maxFractionalDigits=2
    font {
     bForceMonospaced=false
     bMaterialIsStencil=false
     fontMaterial=null
     fontObject=&Roboto
     letterSpacing=0
     monospacedWidth=1
     size=9
     skewAmount=0
     typefaceFontName=Regular
     outlineSettings {
      bApplyOutlineToDropShadows=false
      bMiteredCorners=false
      bSeparateFillAlpha=false
      outlineMaterial=null
      outlineSize=0
      outlineColor=<rgba 0 0 0 1>
     }
    }
    foregroundColor {colorUseRule=UseColor_Specified, specifiedColor=<rgba 1 1 1 1>}
   }
  }]
}
