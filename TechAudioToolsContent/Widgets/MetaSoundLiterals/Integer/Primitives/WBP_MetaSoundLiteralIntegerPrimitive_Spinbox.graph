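# Spinbox widget primitive from TechAudioTools for editing a MetaSound integer literal, built on a single EditorUtilitySpinBox that snaps to whole numbers.
# Value changes are truncated to int and written to SourceValue on the integer literal viewmodel.
# It reports its Input Name and uses it to pick its MetaSound input viewmodel from the supplied map.
# Slider begin/end re-fire as dispatchers, and mouse enter/leave broadcast an Is Hovered field change.
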WidgetBlueprint /TechAudioToolsContent/Widgets/MetaSoundLiterals/Integer/Primitives/WBP_MetaSoundLiteralIntegerPrimitive_Spinbox
graphs [
 EventGraph {nodes [
   n1 = "On Value Changed (Spinbox)"!() -> (OutputDelegate, then, InValue:float) {is=title K2Node_ComponentBoundEvent}
   > Set SourceValue!(
    execute
    SourceValue:int=Truncate(A=@n1.InValue)
    self:Meta Sound Literal Integer Viewmodel Object Reference=Get MetaSoundLiteralViewModel_Integer() -> (MetaSoundLiteralViewModel_Integer:Meta Sound Literal Integer Viewmodel Object Reference)
   ) -> (then, Output_Get:int=0)
   "On Begin Slider Movement (Spinbox)"() {is=title K2Node_ComponentBoundEvent}
   > Call On Begin Slider Movement!(execute:Exec, self:Meta Sound Literal Integer Primitive Spinbox Object Reference) -> (then:Exec) {is=title K2Node_CallDelegate}
   "On End Slider Movement (Spinbox)"!() -> (OutputDelegate, then, InValue:float) {is=title K2Node_ComponentBoundEvent}
   > Call On End Slider Movement!(execute:Exec, self:Meta Sound Literal Integer Primitive Spinbox Object Reference) -> (then:Exec) {is=title K2Node_CallDelegate}
   n2 = Event Set Input Viewmodels!() -> (OutputDelegate, then, InputViewModels:Map of Names to Meta Sound Input Viewmodel Object References) {is=title K2Node_Event}
   > SetMetaSoundInputViewModel!(
    execute
    self
    Viewmodel:Meta Sound Input Viewmodel Object Reference=Find(
     TargetMap:Map of Names to Meta Sound Input Viewmodel Object References=@n2.InputViewModels
     Key:"Name (by ref)"=Get Input Name() -> (Input Name:Name=None)
    ) -> (Value:Meta Sound Input Viewmodel Object Reference)
   ) -> (then)
   n3 = Event OnMouseEnter()
   Broadcast Field Value Changed!(execute=[@n3.then, @n4.then], self:Widget Object Reference, FieldId:Field Notification Id Structure=(FieldName="Is Hovered")) -> (then)
   n4 = Event OnMouseLeave!() -> (OutputDelegate, then, MouseEvent:"Pointer Event Structure (by ref)") {is=title K2Node_Event}
  ]}
 GetInputViewModelNames {nodes [
   Get Input Viewmodel Names() {is=title K2Node_FunctionEntry}
   > Return Node!(execute, ReturnValue:Array of Names=Make Array("[0]":Name=Get Input Name() -> (Input Name=None)) -> (Array:Array of Names))
  ]}
 Is Hovered {nodes [
   Is Hovered() {is=title K2Node_FunctionEntry}
   > Return Node(ReturnValue:bool=IsHovered(self=Get Spinbox() -> (Spinbox:Editor Utility Spin Box Object Reference)))
  ]}
 On Begin Slider Movement {nodes [On Begin Slider Movement() {is=title K2Node_FunctionEntry}]}
 On End Slider Movement {nodes [On End Slider Movement() {is=title K2Node_FunctionEntry}]}
]
variables [
 Input Name {type=string, default=None}
 On Begin Slider Movement {}
 On End Slider Movement {}
 Spinbox Style {
  type=SpinBoxStyle
  default {
   activeBackgroundBrush {
    bIsDynamicallyLoaded=false
    drawAs=RoundedBox
    imageType=NoImage
    mirroring=NoMirror
    resourceName=None
    resourceObject=null
    tiling=NoTile
    imageSize=<0 0>
    margin=<ltrb 0 0 0 0>
    outlineSettings {
     bUseBrushTransparency=false
     roundingType=FixedRadius
     width=1
     color {colorUseRule=UseColor_ColorTable, specifiedColor=<rgba 0 0 0 0>}
     cornerRadii=<4 4 4 4>
    }
    tintColor {colorUseRule=UseColor_Specified, specifiedColor=<rgba 0.0048 0.0048 0.0048 1>}
    uVRegion {bIsValid=false, max=<0 0>, min=<0 0>}
   }
   activeFillBrush {
    bIsDynamicallyLoaded=false
    drawAs=RoundedBox
    imageType=NoImage
    mirroring=NoMirror
    resourceName=None
    resourceObject=null
    tiling=NoTile
    imageSize=<0 0>
    margin=<ltrb 0 0 0 0>
    outlineSettings {
     bUseBrushTransparency=false
     roundingType=FixedRadius
     width=1
     color {colorUseRule=UseColor_Specified, specifiedColor=<rgba 0 0 0 0>}
     cornerRadii=<4 4 4 4>
    }
    tintColor {colorUseRule=UseColor_Specified, specifiedColor=<rgba 0 0.162 0.7454 1>}
    uVRegion {bIsValid=false, max=<0 0>, min=<0 0>}
   }
   arrowsImage {
    bIsDynamicallyLoaded=false
    drawAs=NoDrawType
    imageType=NoImage
    mirroring=NoMirror
    resourceName=None
    resourceObject=null
    tiling=NoTile
    imageSize=<0 0>
    margin=<ltrb 0 0 0 0>
    outlineSettings {
     bUseBrushTransparency=false
     roundingType=HalfHeightRadius
     width=0
     color {colorUseRule=UseColor_Specified, specifiedColor=<rgba 0 0 0 0>}
     cornerRadii=<0 0 0 0>
    }
    tintColor {colorUseRule=UseColor_Specified, specifiedColor=<rgba 1 1 1 1>}
    uVRegion {bIsValid=false, max=<0 0>, min=<0 0>}
   }
   backgroundBrush {
    bIsDynamicallyLoaded=false
    drawAs=RoundedBox
    imageType=NoImage
    mirroring=NoMirror
    resourceName=None
    resourceObject=null
    tiling=NoTile
    imageSize=<0 0>
    margin=<ltrb 0 0 0 0>
    outlineSettings {
     bUseBrushTransparency=false
     roundingType=FixedRadius
     width=1
     color {colorUseRule=UseColor_ColorTable, specifiedColor=<rgba 0 0 0 0>}
     cornerRadii=<4 4 4 4>
    }
    tintColor {colorUseRule=UseColor_Specified, specifiedColor=<rgba 0.0048 0.0048 0.0048 1>}
    uVRegion {bIsValid=false, max=<0 0>, min=<0 0>}
   }
   foregroundColor {colorUseRule=UseColor_Specified, specifiedColor=<rgba 1 1 1 1>}
   hoveredBackgroundBrush {
    bIsDynamicallyLoaded=false
    drawAs=RoundedBox
    imageType=NoImage
    mirroring=NoMirror
    resourceName=None
    resourceObject=null
    tiling=NoTile
    imageSize=<0 0>
    margin=<ltrb 0 0 0 0>
    outlineSettings {
     bUseBrushTransparency=false
     roundingType=FixedRadius
     width=1
     color {colorUseRule=UseColor_ColorTable, specifiedColor=<rgba 0 0 0 0>}
     cornerRadii=<4 4 4 4>
    }
    tintColor {colorUseRule=UseColor_Specified, specifiedColor=<rgba 0.0048 0.0048 0.0048 1>}
    uVRegion {bIsValid=false, max=<0 0>, min=<0 0>}
   }
   hoveredFillBrush {
    bIsDynamicallyLoaded=false
    drawAs=RoundedBox
    imageType=NoImage
    mirroring=NoMirror
    resourceName=None
    resourceObject=null
    tiling=NoTile
    imageSize=<0 0>
    margin=<ltrb 0 0 0 0>
    outlineSettings {
     bUseBrushTransparency=false
     roundingType=FixedRadius
     width=1
     color {colorUseRule=UseColor_Specified, specifiedColor=<rgba 0 0 0 0>}
     cornerRadii=<4 4 4 4>
    }
    tintColor {colorUseRule=UseColor_Specified, specifiedColor=<rgba 0.0953 0.0953 0.0953 1>}
    uVRegion {bIsValid=false, max=<0 0>, min=<0 0>}
   }
   inactiveFillBrush {
    bIsDynamicallyLoaded=false
    drawAs=RoundedBox
    imageType=NoImage
    mirroring=NoMirror
    resourceName=None
    resourceObject=null
    tiling=NoTile
    imageSize=<0 0>
    margin=<ltrb 0 0 0 0>
    outlineSettings {
     bUseBrushTransparency=false
     roundingType=FixedRadius
     width=1
     color {colorUseRule=UseColor_Specified, specifiedColor=<rgba 0 0 0 0>}
     cornerRadii=<4 4 4 4>
    }
    tintColor {colorUseRule=UseColor_Specified, specifiedColor=<rgba 0.0395 0.0395 0.0395 1>}
    uVRegion {bIsValid=false, max=<0 0>, min=<0 0>}
   }
   insetPadding=<ltrb 1 1 1 1>
   textPadding=<ltrb 7 2.5 7 2.5>
  }
 }
]
widgets {
 info {inheritedWidgetCount=0, parentClass=&UserWidget, rootWidgetClass=/Script/Blutility.EditorUtilitySpinBox, widgetCount=1}
 widgets [Spinbox {
   widgetClassPath=&EditorUtilitySpinBox
   bIsVariable=true
   widget=Spinbox
   properties {
    bAlwaysUsesDeltaSnap=true
    delta=1
    justification=Center
    maxFractionalDigits=0
    minFractionalDigits=0
    font {
     bForceMonospaced=false
     bMaterialIsStencil=false
     fontMaterial=null
     fontObject=&Roboto
     letterSpacing=0
     monospacedWidth=1
     size=9
     skewAmount=0
     typefaceFontName=Regular
     outlineSettings {
      bApplyOutlineToDropShadows=false
      bMiteredCorners=false
      bSeparateFillAlpha=false
      outlineMaterial=null
      outlineSize=0
      outlineColor=<rgba 0 0 0 1>
     }
    }
    foregroundColor {colorUseRule=UseColor_Specified, specifiedColor=<rgba 1 1 1 1>}
   }
  }]
}
