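# Editor utility MetaSound preset widget for mixing and auditioning first-person weapon fire presets based on the Complete and Simple base sources.
# Construct binds volume, fire rate, burst and ammo widgets to the preset builder through MetaSound viewmodels.
# Picking a weapon definition fills timing, ammo and burst fields from its stats row.
# Fire, ADS and ammo controls drive the audition component's parameters, as do channel mute and solo, converting dB to linear gain.
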EditorUtilityWidgetBlueprint /FortMetaSoundWeapon/Items/Metasounds/Widgets/EUW_MetaSoundPresetWidget_WeaponFire
graphs [
 EventGraph {nodes [
   n1 = "On Check State Changed (IsBurstCheckBox)"!() -> (OutputDelegate, then, bIsChecked:bool) {is=title K2Node_ComponentBoundEvent}
   > n2 = Set bSetBurstFire(bSetBurstFire:bool=@n1.bIsChecked) -> (Output_Get:bool)
   > SetVisibility(self=Get BurstSettings() -> (BurstSettings:Border Object Reference), InVisibility=Select(Option 0=Collapsed, Option 1=Visible, Index=@n2))
   n3 = "On Checked State Changed (AutoFire)"!() -> (OutputDelegate, then, Is Checked:bool) {is=title K2Node_ComponentBoundEvent}
   > n4 = Get!(execute:Exec) -> (then:Exec, else:Exec, AuditionAudioComponent:Audio Component Object Reference)
   > Set Boolean Parameter!(execute, self:Audio Parameter Controller Interface Interface=@n4, InName:Name=bIsFiring, InBool:bool=@n3.Is Checked) -> (then) {is=title K2Node_CallFunction}
   n5 = "On Check State Changed (ADSCheckBox)"!() -> (OutputDelegate, then, bIsChecked:bool) {is=title K2Node_ComponentBoundEvent}
   > n6 = Get()
   > Set Boolean Parameter(self=@n6, InName=Weapon.IsAimingDownSights, InBool=@n5.bIsChecked)
   "On Pressed (SingleFire)"() {is=title K2Node_ComponentBoundEvent}
   > n7 = Get()
   > Execute Trigger Parameter!(
    execute
    self:Audio Parameter Controller Interface Interface=@n7
    InName:Name=SelectName!(A:Name=FireBurst, B:Name=FireSingleShot, bPickA:bool=Get bSetBurstFire() -> (bSetBurstFire:bool)) -> (ReturnValue:Name=None)
   ) -> (then)
   "On Property Changed (WeaponWID)"!() -> (OutputDelegate, then, PropertyName:Name=None) {is=title K2Node_ComponentBoundEvent}
   > n8 = Get!(execute) -> (then, else, WeaponItemDefinition:Fort Weapon Item Definition Object Reference)
   n9 = GetWeaponStatHandle!(self:Fort Weapon Item Definition Object Reference=@n8) -> (ReturnValue_DataTable:Data Table Object Reference, ReturnValue_RowName:Name)
   > n10 = Get Data Table Row!(execute:Exec, DataTable:Data Table Object Reference=@n9, RowName:Name=@n9.ReturnValue_RowName) -> (then:Exec, RowNotFound:Exec, ReturnValue:Fort Ranged Weapon Stats Structure)
   n11 = Break Fort Ranged Weapon Stats!(FortRangedWeaponStats:"Fort Ranged Weapon Stats Structure (by ref)"=@n10) -> (
    Spread:float=~
    OverrideBaseSpread:float=~
    SpreadDownsights:float=~
    JumpingFallingSpreadDownsights:float=~
    OverrideSpreadDownsights:float=~
    StandingStillSpreadMultiplier:float=~
    AthenaCrouchingSpreadMultiplier:float=~
    AthenaSlidingSpreadMultiplier:float=~
    AthenaJumpingFallingSpreadMultiplier:float=~
    AthenaSprintingSpreadMultiplier:float=~
    MinSpeedForSpreadMultiplier:float=~
    MaxSpeedForSpreadMultiplier:float=~
    SpreadDownsightsAdditionalCooldownTime:float=~
    SpreadExponent:float=~
    HeatX1:float=~
    HeatY1:float=~
    HeatX2:float=~
    HeatY2:float=~
    HeatX3:float=~
    HeatY3:float=~
    HeatXScale:float=~
    HeatYScale:float=~
    CoolX1:float=~
    CoolY1:float=~
    CoolX2:float=~
    CoolY2:float=~
    CoolX3:float=~
    CoolY3:float=~
    CoolXScale:float=~
    CoolYScale:float=~
    PerfectAimCooldown:float=~
    bUseConstantCooldownSpeed:bool=~
    ConstantCooldownSpeed:float=~
    bUseConstantCooldownRecoverySpeed:bool=~
    ConstantCooldownRecoverySpeed:float=~
    bUseMovementCooldownOverride:bool=~
    BulletsPerCartridge:int
    MinHitsPerCartridge:int
    FiringRate:float=~
    ROFScale:float=~
    BurstFiringRate:float=~
    FiringRateDownsightsMultiplier:float=~
    EarlyRefireRate:float=~
    EarlyRefireSpreadPenalty:float=~
    EarlyRefireSpreadMinimumMultiplier:float=~
    EarlyRefireSpreadDownsightsMinimumMultiplier:float=~
    RecoilVert:float=~
    RecoilVertScale:float=~
    RecoilVertScaleGamepad:float=~
    VertRecoilDownChance:float=~
    RecoilHoriz:float=~
    RecoilHorizScale:float=~
    RecoilHorizScaleGamepad:float=~
    RecoilInterpSpeed:float=~
    RecoilRecoveryInterpSpeed:float=~
    RecoilRecoveryDelay:float=~
    RecoilRecoveryEaseInTime:float=~
    RecoilRecoveryFraction:float=~
    RecoilDownsightsMultiplier:float=~
    RecoilStandingStillMultiplier:float=~
    RecoilMovingMultiplier:float=~
    AthenaRecoilMagnitudeMin:float=~
    AthenaRecoilMagnitudeMax:float=~
    AthenaRecoilMagnitudeScale:float=~
    AthenaRecoilAngleMin:float=~
    AthenaRecoilAngleMax:float=~
    AthenaRecoilInterpSpeed:float=~
    AthenaRecoilRecoveryInterpSpeed:float=~
    AthenaRecoilDownsightsMultiplier:float=~
    AthenaRecoilHipFireMultiplier:float=~
    PitchLimitForPerBulletRecoil:float=~
    bPerBulletRecoilRecoveryWhileFiring:bool=~
    bEvenRecoilWhenCountered:bool=~
    bUseOffsetRecoil:bool=~
    OffsetRecoilFactor:float=~
    RecoilBounceFactor:float=~
    RecoilBounceOffsetFactor:float=~
    OffsetRecoilLimitOverride:float=~
    AthenaAimAssistRange:float=~
    ADSTransitionInTime:float=~
    ADSTransitionOutTime:float=~
    ToggleScopedViewTransitionTime:float=~
    MaxSpareAmmo:int
    BulletsPerTracer:int
    AIDelayBeforeFiringMin:float=~
    AIDelayBeforeFiringMax:float=~
    AIFireDurationMin:float=~
    AIFireDurationMax:float=~
    AIMinSpreadDuration:float=~
    AIMaxSpreadDuration:float=~
    AIDurationSpreadMultiplier:float=~
    AIAdditionalSpreadForTargetMovingLaterally:float=~
    AIAthenaHearFiringNoiseRange:Scalable Float Structure
    AIAthenaHearLightweightProjectileImpactNoiseRange:Scalable Float Structure
    AIAthenaHearLightweightProjectileFlyByNoiseRange:Scalable Float Structure
    AIAthenaLightweightProjectileFlyByEmissionInterval:Scalable Float Structure
    EQSDensity:float=~
    MinApproachRange:float=~
    MinActualRange:float=~
    MinPreferredRange:float=~
    MinPreferredRangeEQS:float=~
    MaxPreferredRangeEQS:float=~
    MaxPreferredRange:float=~
    MaxActualRange:float=~
    MaxApproachRange:float=~
    RangeToAutomaticallyAddEnemyPawnGoals:float=~
    SweepRadius:float=~
    AutoReloadDelayOverride:float=~
    OverheatingMaxValue:float=~
    OverheatHeatingValue:float=~
    FullChargeOverheatHeatingValue:float=~
    OverheatingCoolingValue:float=~
    FullyOverheatedCoolingValue:float=~
    HeatingCooldownDelay:float=~
    OverheatedCooldownDelay:float=~
    bCoolOverheatWhileCharging:bool=~
    bCoolOverheatWhileUnequipped:bool=~
    ManualVentingMinOverheatValueRequirement:float=~
    FortHomingTurnSpeedMin:float=~
    FortHomingTurnSpeedMax:float=~
    FortHomingTimeUntilMaxTurnSpeed:float=~
    ProjectileVelocity:float=~
    BulletsPerBulletFireFX:int
    MaxDamagePerCartridge:float=~
    bUseConstantLerpForRecoil:bool=~
    bStartRecoilDelayAfterRecoilPeaks:bool=~
    bUseAccumulatedRecoilToSelectShotIndex:bool=~
    RecoilFractionToBulletCountCurve:Curve Float Soft Object Reference
    bUseAccumulatedRecoilToLerpMinToMaxPerBulletRecoilValues:bool=~
    RecoilMinToMaxCurve:Curve Float Soft Object Reference
    RecoilMinVerticalCurve:Curve Float Soft Object Reference
    RecoilMaxVerticalCurve:Curve Float Soft Object Reference
    RecoilMinHorizontalCurve:Curve Float Soft Object Reference
    RecoilMaxHorizontalCurve:Curve Float Soft Object Reference
    RecoilErrorAddedPerShot:float=~
    RecoilOffsetConsideredMaxFraction:float=~
    RecoilDelay:float=~
    bDelaySpreadWithRecoil:bool=~
    bDisableOneFrameFiringAnimationDelay:bool=~
    bUseScopedPerBulletRecoilOverridesForBasicADS:bool=~
    bLerpSprintingSpreadMultiplierWithMovementSpeed:bool=~
    SpreadMultiplierCap:float=~
    PerBulletRecoil:Array of Fort Ranged Weapon Recoil Structures
    ScopedPerBulletRecoilOverride:Array of Fort Ranged Weapon Recoil Structures
    PerBulletSpread:Array of Fort Ranged Weapon Spread Structures
    OverrideBulletPattern:Array of Fort Bullet Pattern Entry Data Structures
    BaseLevel:int
    NamedWeightRow:Name
    DmgPB:float=~
    DmgMid:float=~
    DmgLong:float=~
    DmgMaxRange:float=~
    EnvDmgPB:float=~
    EnvDmgMid:float=~
    EnvDmgLong:float=~
    EnvDmgMaxRange:float=~
    ImpactDmgPB:float=~
    ImpactDmgMid:float=~
    ImpactDmgLong:float=~
    ImpactDmgMaxRange:float=~
    bForceControl:bool=~
    RngPB:float=~
    RngMid:float=~
    RngLong:float=~
    RngMax:float=~
    DmgScaleTable:Curve Table Soft Object Reference
    DmgScaleTableRow:Name
    DmgScale:float=~
    EnvDmgScaleTable:Curve Table Soft Object Reference
    EnvDmgScaleTableRow:Name
    EnvDmgScale:float=~
    ImpactDmgScaleTable:Curve Table Soft Object Reference
    ImpactDmgScaleTableRow:Name
    ImpactDmgScale:float=~
    SurfaceRatioRowName:Name
    DamageZone_Light:float=~
    DamageZone_Normal:float=~
    DamageZone_Critical:float=~
    DamageZone_Vulnerability:float=~
    DamageZone_Invulnerability:float=~
    KnockbackMagnitude:float=~
    MidRangeKnockbackMagnitude:float=~
    LongRangeKnockbackMagnitude:float=~
    KnockbackZAngle:float=~
    ShortRangeHitImpulseMagnitude:float=~
    MidRangeHitImpulseMagnitude:float=~
    LongRangeHitImpulseMagnitude:float=~
    HitImpulseZBias:float=~
    StunTime:float=~
    StunScale:float=~
    Durability:Data Table Soft Object Reference
    DurabilityRowName:Name
    DurabilityScale:float=~
    DurabilityPerUse:float=~
    DurabilityCooldown:Data Table Soft Object Reference
    DurabilityCooldownRowName:Name
    DurabilityCooldownScale:float=~
    FullChargeDurabilityPerUse:float=~
    DiceCritChance:float=~
    DiceCritDamageMultiplier:float=~
    ReloadTime:float=~
    ReloadScale:float=~
    ReloadType:EFortWeaponReloadType Enum
    bAllowReloadInterrupt:bool=~
    bReloadInterruptIsImmediate:bool=~
    NumIndividualBulletsToReload:int
    ClipSize:int
    ClipScale:float=~
    InitialClips:int
    bShouldAddInitialAmmo:bool=~
    CartridgePerFire:int
    AmmoCostPerFire:int
    MaxAmmoCostPerFire:int
    MinChargeTime:float=~
    MaxChargeTime:float=~
    ChargeDownTime:float=~
    bAutoDischarge:bool=~
    MaxChargeTimeUntilDischarge:float=~
    MinChargeDamageMultiplier:float=~
    MaxChargeDamageMultiplier:float=~
    ChargeDamageMultiplierCurve:Curve Float Soft Object Reference
    EquipAnimRate:float=~
    StowAnimRate:float=~
    QuickBarSlotCooldownDuration:float=~
    AimAssistPullStrengthMultiplier:float=~
    AimAssistSlowStrengthMultiplier:float=~
    AimAssistAutoAimSize:float=~
    MeleeAimAssistRequiredTag:Gameplay Tag Structure
    MaxUnderReticleDistance:float=~
    AutofireRange:float=~
    AutofireMinRange:float=~
    AutofireZoomTargetMinSizePercent:float=~
    AutofireAcquisitionDelay:float=~
    AutofireDBNOAcquisitionDelay:float=~
    AutofireAcquisitionRechargeTime:float=~
    AutofireMinChargingTime:float=~
    AutofireReleaseTime:float=~
    AutofireCooldown:float=~
    PerWeaponHotfixes:Fort Per Weapon Hotfixes Structure
   ) {is=title K2Node_BreakStruct}
   > Set Value!(
    execute
    Value:float=SafeDivide(A=1, B=@n11.FiringRate)
    self:Spin Box Object Reference=Get Spinbox!(self:Meta Sound Literal Float Primitive Spinbox Object Reference=Get FireRateEntry() -> (FireRateEntry:Meta Sound Literal Float Primitive Spinbox Object Reference)) -> (Spinbox:Editor Utility Spin Box Object Reference)
   ) -> (then, Output_Get:float)
   > Set Value(
    Value=@n11.BurstFiringRate
    self=Get Spinbox(self=Get BurstFireRateEntry() -> (BurstFireRateEntry:Meta Sound Literal Float Primitive Spinbox Object Reference))
   )
   > Set Value(
    Value="To Float (Integer)"(InInt=@n11.ClipSize)
    self=Get Spinbox(self:Meta Sound Literal Integer Primitive Spinbox Object Reference=Get AmmoCountEntry() -> (AmmoCountEntry:Meta Sound Literal Integer Primitive Spinbox Object Reference))
   )
   > Set Value(
    Value="To Float (Integer)"(InInt=@n11.CartridgePerFire)
    self=Get Spinbox(self=Get BurstCountEntry() -> (BurstCountEntry:Meta Sound Literal Integer Primitive Spinbox Object Reference))
   )
   n12 = "integer > integer"(A=@n11.CartridgePerFire, B=1)
   > SetIsChecked!(
    execute
    self:Check Box Object Reference=Get IsBurstCheckBox() -> (IsBurstCheckBox:Editor Utility Check Box Object Reference)
    InIsChecked:bool=@n12
   ) -> (then)
   > SetVisibility(self=Get BurstSettings(), InVisibility=Select(Option 0=Collapsed, Option 1=Visible, Index=@n12))
   SetIsChecked(execute=@n8.else, self=Get IsBurstCheckBox(), InIsChecked=false)
   > SetVisibility(self=Get BurstSettings(), InVisibility=Collapsed)
   n13 = "On Check State Changed (UseAmmoCheckBox)"!() -> (OutputDelegate, then, bIsChecked:bool) {is=title K2Node_ComponentBoundEvent}
   > n14 = Get!(execute) -> (then, else, AuditionAudioComponent:Audio Component Object Reference)
   > Set Boolean Parameter(self=@n14, InName=bHasUnlimitedAmmo, InBool=NOT Boolean(A=@n13.bIsChecked))
   InitializeLiteralWidgets() {is=title K2Node_CustomEvent}
   > n15 = For Each Loop(Array:Array of Meta Sound Literal Widget Interface Interfaces=Make Array!(
     "[0]":Meta Sound Literal Widget Interface Interface=Get MainFireVolume() -> (MainFireVolume:EUW Channel Weapon Object Reference)
     "[1]":Meta Sound Literal Widget Interface Interface=Get MechVolume() -> (MechVolume:EUW Channel Weapon Object Reference)
     "[2]":Meta Sound Literal Widget Interface Interface=Get LowAmmoVolume() -> (LowAmmoVolume:EUW Channel Weapon Object Reference)
     "[3]":Meta Sound Literal Widget Interface Interface=Get SweetenerVolume() -> (SweetenerVolume:EUW Channel Weapon Object Reference)
     "[4]":Meta Sound Literal Widget Interface Interface=Get OnStopVolume() -> (OnStopVolume:EUW Channel Weapon Object Reference)
     "[5]":Meta Sound Literal Widget Interface Interface=Get OutputVolume() -> (OutputVolume:EUW Output Weapon Object Reference)
     "[6]":Meta Sound Literal Widget Interface Interface=Get FireRateEntry()
     "[7]":Meta Sound Literal Widget Interface Interface=Get BurstFireRateEntry()
     "[8]":Meta Sound Literal Widget Interface Interface=Get BurstCountEntry()
     "[9]":Meta Sound Literal Widget Interface Interface=Get AmmoCountEntry()
     "[10]":Meta Sound Literal Widget Interface Interface=Get SingleVoice() -> (SingleVoice:Meta Sound Literal Bool Primitive Check Box Object Reference)
    ) -> (Array:Array of Meta Sound Literal Widget Interface Interfaces)) -> (Array Element:Meta Sound Literal Widget Interface Interface)
   > Set Literal Widget Input Viewmodels!(
    execute
    Literal Widget:Meta Sound Literal Widget Interface Interface=@n15
    MetaSound Viewmodel:Meta Sound Viewmodel Object Reference=Get MetaSoundEditorViewModel() -> (MetaSoundEditorViewModel:Meta Sound Editor Viewmodel Object Reference)
    __WorldContext:Object Reference=Self-Reference()
   ) -> (then) {is=title K2Node_CallFunction}
   n16 = Event On MetaSound Audition State Changed!() -> (OutputDelegate, then, AudioComponent:Audio Component Object Reference, bIsAuditioning:bool) {is=title K2Node_Event}
   > n17 = Set AuditionAudioComponent(AuditionAudioComponent:Audio Component Object Reference=@n16.AudioComponent) -> (Output_Get:Audio Component Object Reference)
   n18 = Get UseAmmoCheckBox() -> (UseAmmoCheckBox:Editor Utility Check Box Object Reference)
   > Set Boolean Parameter(self=@n17, InName=bHasUnlimitedAmmo, InBool=NOT Boolean(A=IsChecked(self=@n18)))
   > Branch(Condition=@n16.bIsAuditioning)
   > n19 = Get()
   > Set Boolean Parameter(
    self=@n19
    InName=bIsFiring
    InBool=Is Checked!(self:Tech Audio Tools Toggle Button Content Slot Object Reference=Get AutoFire() -> (AutoFire:Tech Audio Tools Toggle Button Content Slot Object Reference)) -> (Return Value:bool)
   )
   Event Construct()
   > Initialize Builder!(
    execute
    self:Meta Sound Viewmodel Object Reference=Get MetaSoundEditorViewModel()
    InBuilder:Meta Sound Builder Base Object Reference=Get Builder() -> (Builder:Meta Sound Builder Base Object Reference)
   ) -> (then)
   n20 = Get IsBurstCheckBox()
   > SetVisibility(self=Get BurstSettings(), InVisibility=Select(Option 0=Collapsed, Option 1=Visible, Index=IsChecked(self=@n20)))
   > InitializeLiteralWidgets() {is=title K2Node_CallFunction}
   > SetupBindings()
   > Clear(TargetArray:Array of Names=Get NamesToSolo() -> (NamesToSolo:Array of Names))
   > Clear!(
    execute
    TargetMap:Map of Names to EUW Channel Weapon Object References=Get TChannelArray() -> (TChannelArray:Map of Names to EUW Channel Weapon Object References)
   ) -> (then) {is=title K2Node_CallFunction}
   > Set TChannelArray(TChannelArray:Map of Names to EUW Channel Weapon Object References=Make Map!(
     Key 0:Name=MainFire.Volume
     Value 0:EUW Channel Weapon Object Reference=Get MainFireVolume()
     Key 1:Name=Mech.Volume
     Value 1:EUW Channel Weapon Object Reference=Get MechVolume()
     Key 2:Name=LowAmmo.Volume
     Value 2:EUW Channel Weapon Object Reference=Get LowAmmoVolume()
     Key 3:Name=Sweetener.Volume
     Value 3:EUW Channel Weapon Object Reference=Get SweetenerVolume()
     Key 4:Name=OnStopFire.Volume
     Value 4:EUW Channel Weapon Object Reference=Get OnStopVolume()
    ) -> (Map:Map of Names to EUW Channel Weapon Object References) {is=title K2Node_MakeMap}) -> (Output_Get:Map of Names to EUW Channel Weapon Object References)
   Event PreConstruct()
   > SetObject(self=Get WeaponWID() -> (WeaponWID:Single Property View Object Reference), NewObject=Self-Reference())
   n21 = Event On MetaSoundPreset Widget Constructed!() -> (OutputDelegate, then, Builder:Meta Sound Builder Base Object Reference) {is=title K2Node_Event}
   > Set Builder(Builder:Meta Sound Builder Base Object Reference=@n21.Builder) -> (Output_Get:Meta Sound Builder Base Object Reference)
  ]}
 GetEditorName {nodes [
   GetEditorName() {is=title K2Node_FunctionEntry}
   > Return Node!(execute, ReturnValue:Text)
  ]}
 GetIconBrushName {nodes [
   GetIconBrushName() {is=title K2Node_FunctionEntry}
   > Return Node(ReturnValue:Name=None)
  ]}
 GetSupportedMetaSounds {nodes [
   GetSupportedMetaSounds() {is=title K2Node_FunctionEntry}
   > Return Node!(
    execute
    bSupportAllPresets:bool=false
    ExcludedMetaSounds:Array of Meta Sound Document Interface Interfaces
    IncludedMetaSounds:Array of Meta Sound Document Interface Interfaces=Get Included Meta Sounds() -> (Included Meta Sounds:Array of Meta Sound Document Interface Interfaces)
   )
  ]}
 OnInitializingTab {nodes [
   On Initializing Tab() {is=title K2Node_FunctionEntry}
   > Return Node!(execute, TabInfo:Meta Sound Template Widget Tab Info Structure)
  ]}
 OnSliderMute {nodes [
   n1 = OnSliderMute!() -> (then, ParameterName:Name, bIsChecked:bool=~) {is=title K2Node_FunctionEntry}
   n2 = Find(TargetMap:Map of Names to EUW Channel Weapon Object References=Get TChannelArray(), Key:"Name (by ref)"=@n1) -> (Value:EUW Channel Weapon Object Reference)
   n3 = float / float(
    A=Get Value!(self:Spin Box Object Reference=Get Value_Spinbox!(self:EUW Channel Weapon Object Reference=@n2) -> (Value_Spinbox:Editor Utility Spin Box Object Reference)) -> (Value:float)
    B=20
   )
   > Set Float Parameter!(
    execute
    self:Audio Parameter Controller Interface Interface=Get AuditionAudioComponent() -> (AuditionAudioComponent:Audio Component Object Reference)
    InName:Name=@n1
    InFloat:float=SelectFloat(A=0, B=Power!(Base:double=10, Exp:double=@n3) -> (ReturnValue:double) {is=title K2Node_CallFunction}, bPickA=@n1.bIsChecked)
   ) -> (then) {is=title K2Node_CallFunction}
  ]}
 OnSliderSolo {nodes [
   n1 = OnSliderSolo!() -> (then, ParameterName:Name, bIsChecked:bool=~) {is=title K2Node_FunctionEntry}
   > Branch(Condition=Is Not Empty!(TargetArray:Array of Names=Get InputNameArray() -> (InputNameArray:Array of Names)) -> (ReturnValue:bool))
   > n2 = Branch(Condition=@n1.bIsChecked)
   > n3 = Add Unique(TargetArray:Array of Names=Get NamesToSolo(), NewItem:"Name (by ref)"=Get ParameterName() -> (ParameterName:Name=None))
   n4 = For Each Loop(Exec=[@n3.then, @n8.else], Array:Array of Names=Get InputNameArray()) -> (Array Element:Name)
   n5 = Get NamesToSolo()
   n6 = Find(TargetMap=Get TChannelArray(), Key=@n4)
   n7 = float / float(A=Get Value(self=Get Value_Spinbox(self=@n6)), B=20)
   > Set Float Parameter(
    self=Get AuditionAudioComponent()
    InName=@n4
    InFloat=SelectFloat(A=Power(Base=10, Exp=@n7), B=0, bPickA=Contains Item(TargetArray=@n5, ItemToFind=@n4))
   )
   Remove Item!(execute=@n2.else, TargetArray:Array of Names=Get NamesToSolo(), Item:"Name (by ref)"=Get ParameterName() -> (ParameterName=None)) -> (then, ReturnValue:bool)
   > n8 = Branch(Condition=Is Empty!(TargetArray:Array of Names=Get NamesToSolo()) -> (ReturnValue:bool))
   n9 = Get Value_Spinbox(self=Get SweetenerVolume())
   n10 = Power(Base=10, Exp=float / float(A=Get Value(self=@n9), B=20))
   n11 = Get Value_Spinbox(self=Get MainFireVolume())
   n12 = Power(Base=10, Exp=float / float(A=Get Value(self=@n11), B=20))
   n13 = Get Value_Spinbox(self=Get LowAmmoVolume())
   n14 = Power(Base=10, Exp=float / float(A=Get Value(self=@n13), B=20))
   n15 = Get Value_Spinbox(self=Get MechVolume())
   n16 = Power(Base=10, Exp=float / float(A=Get Value(self=@n15), B=20))
   n17 = Get Value_Spinbox(self=Get OnStopVolume())
   n18 = Power(Base=10, Exp=float / float(A=Get Value(self=@n17), B=20))
   > Set Parameters!(
    execute
    self:Audio Parameter Controller Interface Interface=Get AuditionAudioComponent()
    InParameters:Array of Audio Parameter Structures=Make Array!(
     "[0]":Audio Parameter Structure=FloatToAudioParameter!(Name:Name=MainFire.Volume, Float:float=@n12) -> (ReturnValue:Audio Parameter Structure)
     "[1]":Audio Parameter Structure=FloatToAudioParameter(Name=LowAmmo.Volume, Float=@n14)
     "[2]":Audio Parameter Structure=FloatToAudioParameter(Name=Mech.Volume, Float=@n16)
     "[3]":Audio Parameter Structure=FloatToAudioParameter(Name=OnStopFire.Volume, Float=@n18)
     "[4]":Audio Parameter Structure=FloatToAudioParameter(Name=Sweetener.Volume, Float=@n10)
    ) -> (Array:Array of Audio Parameter Structures)
   ) -> (then) {is=title K2Node_CallFunction}
  ]}
 SetupBindings {nodes [
   SetupBindings() {is=title K2Node_FunctionEntry}
   > Bind Event to On Mute State Changed(
    self:EUW Channel Weapon Object Reference=[Get LowAmmoVolume(), Get MainFireVolume(), Get OnStopVolume(), Get SweetenerVolume(), Get MechVolume()]
    Delegate=Create Event()
   ) {is=title K2Node_AddDelegate}
   > Bind Event to On Solo State Changed(
    self:EUW Channel Weapon Object Reference=[Get OnStopVolume(), Get MainFireVolume(), Get SweetenerVolume(), Get MechVolume(), Get LowAmmoVolume()]
    Delegate=Create Event()
   ) {is=title K2Node_AddDelegate}
  ]}
]
variables [
 AuditionAudioComponent {type=AudioComponent, default=null}
 bIsADS {type=boolean}
 bSetBurstFire {type=boolean}
 Builder {type=MetaSoundBuilderBase, default=null}
 Included Meta Sounds {
  type="array<object>"
  default [
   MetaSoundSource'/FortMetaSoundWeapon/Items/Metasounds/MSS_Base_Weapon_Fire_1p_Complete'
   MetaSoundSource'/FortMetaSoundWeapon/Items/Metasounds/MSS_Base_Weapon_Fire_1p_Simple'
  ]
 }
 InputNameArray {type="array<string>", default [MainFire.Volume, Mech.Volume, LowAmmo.Volume, Sweetener.Volume, OnStopFire.Volume]}
 NamesToSolo {type="array<string>", default []}
 TChannelArray {type="map<EUW_Channel_Weapon_C>", default {}}
 WeaponItemDefinition {type=FortWeaponItemDefinition, default=null}
]
widgets {
 info {namedSlotCount=1, rootWidgetClass=&StackBox, widgetCount=72}
 widgets [
  StackBox_0 {widgetClassPath=&StackBox, widget=StackBox_0, properties {orientation=Orient_Vertical}}
  EditorUtilityScrollBox_25 {
   widgetClassPath=&EditorUtilityScrollBox
   parent=StackBox_0
   widget=EditorUtilityScrollBox_25
   slot {class=StackBoxSlot, properties {size {sizeRule=Fill, value=1}}}
  }
  Mix {
   widgetClassPath=&HorizontalBox
   parent=EditorUtilityScrollBox_25
   widget=Mix
   slot {class=ScrollBoxSlot, properties {verticalAlignment=VAlign_Top}}
  }
  Border_137 {
   widgetClassPath=&Border
   parent=Mix
   widget=Border_137
   properties {
    background {
     bIsDynamicallyLoaded=false
     drawAs=Image
     imageType=NoImage
     mirroring=NoMirror
     resourceName=None
     resourceObject=null
     tiling=NoTile
     imageSize=<32 32>
     margin=<ltrb 0 0 0 0>
     outlineSettings {
      bUseBrushTransparency=false
      roundingType=HalfHeightRadius
      width=0
      color {colorUseRule=UseColor_Specified, specifiedColor=<rgba 0 0 0 0>}
      cornerRadii=<0 0 0 0>
     }
     tintColor {colorUseRule=UseColor_Specified, specifiedColor=<rgba 0.0103 0.0103 0.0103 1>}
     uVRegion {bIsValid=false, max=<0 0>, min=<0 0>}
    }
    padding=<ltrb 8 8 8 8>
   }
   slot {class=HorizontalBoxSlot, properties {size {sizeRule=Fill, value=1}}}
  }
  TextBlock_43 {
   widgetClassPath=&TextBlock
   parent=Border_137
   widget=TextBlock_43
   properties {text=Mix}
   slot {class=BorderSlot, properties {padding=<ltrb 8 8 8 8>}}
  }
  EditorUtilityScrollBox_43 {
   widgetClassPath=&EditorUtilityScrollBox
   parent=EditorUtilityScrollBox_25
   widget=EditorUtilityScrollBox_43
   properties {orientation=Orient_Horizontal}
   slot {class=ScrollBoxSlot}
  }
  HorizontalBox_43 {
   widgetClassPath=&HorizontalBox
   parent=EditorUtilityScrollBox_43
   widget=HorizontalBox_43
   slot {class=ScrollBoxSlot, properties {size {sizeRule=Fill, value=1}}}
  }
  Border_89 {
   widgetClassPath=&Border
   parent=HorizontalBox_43
   widget=Border_89
   properties {brushColor=<rgba 0.0103 0.0103 0.0103 1>, desiredSizeScale=<1.2914 1>, padding=<ltrb 8 8 8 8>}
   slot {class=HorizontalBoxSlot, properties {size {sizeRule=Fill, value=1}}}
  }
  WrapBox_50 {widgetClassPath=&WrapBox, parent=Border_89, widget=WrapBox_50, slot {class=BorderSlot, properties {padding=<ltrb 8 8 8 8>}}}
  MainFireVolume {
   widgetClassPath=&/EUW_Channel_Weapon
   bIsVariable=true
   parent=WrapBox_50
   widget=MainFireVolume
   properties {
    input Name=MainFire.Volume
    label Text=Main Fire
    metaSoundLiteralViewModel_Float=&self:WidgetTree.MainFireVolume.MetaSoundLiteralViewModel_Float_0
   }
   slot {class=WrapBoxSlot}
  }
  Spacer_59 {widgetClassPath=&Spacer, parent=WrapBox_50, widget=Spacer_59, slot {class=WrapBoxSlot, properties {padding=<ltrb 5 5 5 5>}}}
  MechVolume {
   widgetClassPath=&EUW_Channel_Weapon
   bIsVariable=true
   parent=WrapBox_50
   widget=MechVolume
   properties {
    input Name=Mech.Volume
    label Text=Mech
    metaSoundLiteralViewModel_Float=&self:WidgetTree.MechVolume.MetaSoundLiteralViewModel_Float_0
   }
   slot {class=WrapBoxSlot}
  }
  Spacer {widgetClassPath=&Spacer, parent=WrapBox_50, widget=Spacer, slot {class=WrapBoxSlot, properties {padding=<ltrb 5 5 5 5>}}}
  LowAmmoVolume {
   widgetClassPath=&EUW_Channel_Weapon
   bIsVariable=true
   parent=WrapBox_50
   widget=LowAmmoVolume
   properties {
    input Name=LowAmmo.Volume
    label Text=Low Ammo
    metaSoundLiteralViewModel_Float=&self:WidgetTree.LowAmmoVolume.MetaSoundLiteralViewModel_Float_0
   }
   slot {class=WrapBoxSlot}
  }
  Spacer_1 {widgetClassPath=&Spacer, parent=WrapBox_50, widget=Spacer_1, slot {class=WrapBoxSlot, properties {padding=<ltrb 5 5 5 5>}}}
  SweetenerVolume {
   widgetClassPath=&EUW_Channel_Weapon
   bIsVariable=true
   parent=WrapBox_50
   widget=SweetenerVolume
   properties {
    input Name=Sweetener.Volume
    label Text=Sweetener
    metaSoundLiteralViewModel_Float=&self:WidgetTree.SweetenerVolume.MetaSoundLiteralViewModel_Float_0
   }
   slot {class=WrapBoxSlot}
  }
  Spacer_2 {widgetClassPath=&Spacer, parent=WrapBox_50, widget=Spacer_2, slot {class=WrapBoxSlot, properties {padding=<ltrb 5 5 5 5>}}}
  OnStopVolume {
   widgetClassPath=&EUW_Channel_Weapon
   bIsVariable=true
   parent=WrapBox_50
   widget=OnStopVolume
   properties {
    input Name=OnStopFire.Volume
    label Text=On Stop
    metaSoundLiteralViewModel_Float=&self:WidgetTree.OnStopVolume.MetaSoundLiteralViewModel_Float_0
   }
   slot {class=WrapBoxSlot}
  }
  Spacer_3 {widgetClassPath=&Spacer, parent=WrapBox_50, widget=Spacer_3, slot {class=WrapBoxSlot, properties {padding=<ltrb 15 15 15 15>}}}
  OutputVolume {
   widgetClassPath=&/EUW_Output_Weapon
   bIsVariable=true
   parent=WrapBox_50
   widget=OutputVolume
   properties {
    input Name=MasterVolume
    label Text=Output
    metaSoundLiteralViewModel_Float=&self:WidgetTree.OutputVolume.MetaSoundLiteralViewModel_Float_0
   }
   slot {class=WrapBoxSlot}
  }
  Spacer_14 {widgetClassPath=&Spacer, parent=WrapBox_50, widget=Spacer_14, slot {class=WrapBoxSlot, properties {padding=<ltrb 20 20 20 20>}}}
  Spacer_152 {
   widgetClassPath=&Spacer
   parent=EditorUtilityScrollBox_25
   widget=Spacer_152
   slot {class=ScrollBoxSlot, properties {padding=<ltrb 5 5 5 5>}}
  }
  Preview {
   widgetClassPath=&HorizontalBox
   parent=EditorUtilityScrollBox_25
   widget=Preview
   slot {class=ScrollBoxSlot, properties {verticalAlignment=VAlign_Top}}
  }
  Border_250 {
   widgetClassPath=&Border
   parent=Preview
   widget=Border_250
   properties {
    background {
     bIsDynamicallyLoaded=false
     drawAs=Image
     imageType=NoImage
     mirroring=NoMirror
     resourceName=None
     resourceObject=null
     tiling=NoTile
     imageSize=<32 32>
     margin=<ltrb 0 0 0 0>
     outlineSettings {
      bUseBrushTransparency=false
      roundingType=HalfHeightRadius
      width=0
      color {colorUseRule=UseColor_Specified, specifiedColor=<rgba 0 0 0 0>}
      cornerRadii=<0 0 0 0>
     }
     tintColor {colorUseRule=UseColor_Specified, specifiedColor=<rgba 0.0103 0.0103 0.0103 1>}
     uVRegion {bIsValid=false, max=<0 0>, min=<0 0>}
    }
    padding=<ltrb 8 8 8 8>
   }
   slot {class=HorizontalBoxSlot, properties {size {sizeRule=Fill, value=1}}}
  }
  TextBlock {
   widgetClassPath=&TextBlock
   parent=Border_250
   widget=TextBlock
   properties {text=Audition}
   slot {class=BorderSlot, properties {padding=<ltrb 8 8 8 8>}}
  }
  EditorUtilityScrollBox {
   widgetClassPath=&EditorUtilityScrollBox
   parent=EditorUtilityScrollBox_25
   widget=EditorUtilityScrollBox
   properties {orientation=Orient_Horizontal}
   slot {class=ScrollBoxSlot}
  }
  HorizontalBox {
   widgetClassPath=&HorizontalBox
   parent=EditorUtilityScrollBox
   widget=HorizontalBox
   slot {class=ScrollBoxSlot, properties {size {sizeRule=Fill, value=1}}}
  }
  Border_307 {
   widgetClassPath=&Border
   parent=HorizontalBox
   widget=Border_307
   properties {
    background {
     bIsDynamicallyLoaded=false
     drawAs=Image
     imageType=NoImage
     mirroring=NoMirror
     resourceName=None
     resourceObject=null
     tiling=NoTile
     imageSize=<32 32>
     margin=<ltrb 0 0 0 0>
     outlineSettings {
      bUseBrushTransparency=false
      roundingType=HalfHeightRadius
      width=0
      color {colorUseRule=UseColor_Specified, specifiedColor=<rgba 0 0 0 0>}
      cornerRadii=<0 0 0 0>
     }
     tintColor {colorUseRule=UseColor_Specified, specifiedColor=<rgba 0.0103 0.0103 0.0103 1>}
     uVRegion {bIsValid=false, max=<0 0>, min=<0 0>}
    }
    padding=<ltrb 8 8 8 8>
   }
   slot {class=HorizontalBoxSlot, properties {size {sizeRule=Fill, value=1}}}
  }
  HorizontalBox_66 {
   widgetClassPath=&HorizontalBox
   parent=Border_307
   widget=HorizontalBox_66
   slot {class=BorderSlot, properties {padding=<ltrb 8 8 8 8>}}
  }
  VerticalBox_76 {widgetClassPath=&VerticalBox, parent=HorizontalBox_66, widget=VerticalBox_76, slot {class=HorizontalBoxSlot}}
  Border_364 {
   widgetClassPath=&Border
   parent=VerticalBox_76
   widget=Border_364
   properties {
    background {
     bIsDynamicallyLoaded=false
     drawAs=RoundedBox
     imageType=NoImage
     mirroring=NoMirror
     resourceName=None
     resourceObject=null
     tiling=NoTile
     imageSize=<32 32>
     margin=<ltrb 0 0 0 0>
     outlineSettings {
      bUseBrushTransparency=false
      roundingType=FixedRadius
      width=0
      color {colorUseRule=UseColor_Specified, specifiedColor=<rgba 0 0 0 0>}
      cornerRadii=<12 12 12 12>
     }
     tintColor {colorUseRule=UseColor_Specified, specifiedColor=<rgba 0.0212 0.0212 0.0212 1>}
     uVRegion {bIsValid=false, max=<0 0>, min=<0 0>}
    }
    padding=<ltrb 8 8 8 8>
   }
   slot {class=VerticalBoxSlot, properties {verticalAlignment=VAlign_Top}}
  }
  VerticalBox_116 {
   widgetClassPath=&VerticalBox
   parent=Border_364
   widget=VerticalBox_116
   slot {class=BorderSlot, properties {padding=<ltrb 8 8 8 8>}}
  }
  Spacer_218 {widgetClassPath=&Spacer, parent=VerticalBox_116, widget=Spacer_218, properties {size=<10 10>}, slot {class=VerticalBoxSlot}}
  HorizontalBox_148 {widgetClassPath=&HorizontalBox, parent=VerticalBox_116, widget=HorizontalBox_148, slot {class=VerticalBoxSlot}}
  AutoFire {
   widgetClassPath=/TechAudioToolsContent/Widgets/Editor/ToggleButtons/EUW_ToggleButton_ContentSlot
   bIsVariable=true
   parent=HorizontalBox_148
   widget=AutoFire
   properties {
    checked State=Checked
    toolTipText=Play MetaSound first. Automatically fire the weapon.
    check Box Style {
     checkBoxType=ToggleButton
     backgroundHoveredImage {
      bIsDynamicallyLoaded=false
      drawAs=RoundedBox
      imageType=NoImage
      mirroring=NoMirror
      resourceName=None
      resourceObject=null
      tiling=NoTile
      imageSize=<18 18>
      margin=<ltrb 0 0 0 0>
      outlineSettings {
       bUseBrushTransparency=false
       roundingType=FixedRadius
       width=1
       color {colorUseRule=UseColor_ColorTable, specifiedColor=<rgba 0 0 0 0>}
       cornerRadii=<3 3 3 3>
      }
      tintColor {colorUseRule=UseColor_Specified, specifiedColor=<rgba 0.0048 0.0048 0.0048 1>}
      uVRegion {bIsValid=false, max=<0 0>, min=<0 0>}
     }
     backgroundImage {
      bIsDynamicallyLoaded=false
      drawAs=RoundedBox
      imageType=NoImage
      mirroring=NoMirror
      resourceName=None
      resourceObject=null
      tiling=NoTile
      imageSize=<18 18>
      margin=<ltrb 0 0 0 0>
      outlineSettings {
       bUseBrushTransparency=false
       roundingType=FixedRadius
       width=1
       color {colorUseRule=UseColor_ColorTable, specifiedColor=<rgba 0 0 0 0>}
       cornerRadii=<3 3 3 3>
      }
      tintColor {colorUseRule=UseColor_Specified, specifiedColor=<rgba 0.0048 0.0048 0.0048 1>}
      uVRegion {bIsValid=false, max=<0 0>, min=<0 0>}
     }
     backgroundPressedImage {
      bIsDynamicallyLoaded=false
      drawAs=RoundedBox
      imageType=NoImage
      mirroring=NoMirror
      resourceName=None
      resourceObject=null
      tiling=NoTile
      imageSize=<18 18>
      margin=<ltrb 0 0 0 0>
      outlineSettings {
       bUseBrushTransparency=false
       roundingType=FixedRadius
       width=1
       color {colorUseRule=UseColor_ColorTable, specifiedColor=<rgba 0 0 0 0>}
       cornerRadii=<3 3 3 3>
      }
      tintColor {colorUseRule=UseColor_Specified, specifiedColor=<rgba 0.0048 0.0048 0.0048 1>}
      uVRegion {bIsValid=false, max=<0 0>, min=<0 0>}
     }
     borderBackgroundColor {colorUseRule=UseColor_Specified, specifiedColor=<rgba 1 1 1 1>}
     checkedForeground {colorUseRule=UseColor_Specified, specifiedColor=<rgba 1 1 1 1>}
     checkedHoveredForeground {colorUseRule=UseColor_Specified, specifiedColor=<rgba 1 1 1 1>}
     checkedHoveredImage {
      bIsDynamicallyLoaded=false
      drawAs=RoundedBox
      imageType=NoImage
      mirroring=NoMirror
      resourceName=None
      resourceObject=null
      tiling=NoTile
      imageSize=<0 0>
      margin=<ltrb 0 0 0 0>
      outlineSettings {
       bUseBrushTransparency=false
       roundingType=FixedRadius
       width=0
       color {colorUseRule=UseColor_Foreground, specifiedColor=<rgba 0 0 0 0>}
       cornerRadii=<4 4 4 4>
      }
      tintColor {colorUseRule=UseColor_Specified, specifiedColor=<rgba 0.0044 0.2384 1 1>}
      uVRegion {bIsValid=false, max=<0 0>, min=<0 0>}
     }
     checkedImage {
      bIsDynamicallyLoaded=false
      drawAs=RoundedBox
      imageType=NoImage
      mirroring=NoMirror
      resourceName=None
      resourceObject=null
      tiling=NoTile
      imageSize=<0 0>
      margin=<ltrb 0 0 0 0>
      outlineSettings {
       bUseBrushTransparency=false
       roundingType=FixedRadius
       width=0
       color {colorUseRule=UseColor_Foreground, specifiedColor=<rgba 0 0 0 0>}
       cornerRadii=<4 4 4 4>
      }
      tintColor {colorUseRule=UseColor_Specified, specifiedColor=<rgba 0 0.162 0.7454 1>}
      uVRegion {bIsValid=false, max=<0 0>, min=<0 0>}
     }
     checkedPressedForeground {colorUseRule=UseColor_Specified, specifiedColor=<rgba 1 1 1 1>}
     checkedPressedImage {
      bIsDynamicallyLoaded=false
      drawAs=RoundedBox
      imageType=NoImage
      mirroring=NoMirror
      resourceName=None
      resourceObject=null
      tiling=NoTile
      imageSize=<0 0>
      margin=<ltrb 0 0 0 0>
      outlineSettings {
       bUseBrushTransparency=false
       roundingType=FixedRadius
       width=0
       color {colorUseRule=UseColor_Foreground, specifiedColor=<rgba 0 0 0 0>}
       cornerRadii=<4 4 4 4>
      }
      tintColor {colorUseRule=UseColor_Specified, specifiedColor=<rgba 0.0625 0.0625 0.0625 1>}
      uVRegion {bIsValid=false, max=<0 0>, min=<0 0>}
     }
     checkedSlateSound {resourceObject=null}
     foregroundColor {colorUseRule=UseColor_Specified, specifiedColor=<rgba 1 1 1 1>}
     hoveredForeground {colorUseRule=UseColor_Specified, specifiedColor=<rgba 1 1 1 1>}
     hoveredSlateSound {resourceObject=null}
     padding=<ltrb 8 5 8 5>
     pressedForeground {colorUseRule=UseColor_Specified, specifiedColor=<rgba 1 1 1 1>}
     uncheckedHoveredImage {
      bIsDynamicallyLoaded=false
      drawAs=RoundedBox
      imageType=NoImage
      mirroring=NoMirror
      resourceName=None
      resourceObject=null
      tiling=NoTile
      imageSize=<0 0>
      margin=<ltrb 0 0 0 0>
      outlineSettings {
       bUseBrushTransparency=false
       roundingType=FixedRadius
       width=0
       color {colorUseRule=UseColor_Specified, specifiedColor=<rgba 0.0395 0.0395 0.0395 1>}
       cornerRadii=<4 4 4 4>
      }
      tintColor {colorUseRule=UseColor_Specified, specifiedColor=<rgba 0.0953 0.0953 0.0953 1>}
      uVRegion {bIsValid=false, max=<0 0>, min=<0 0>}
     }
     uncheckedImage {
      bIsDynamicallyLoaded=false
      drawAs=RoundedBox
      imageType=NoImage
      mirroring=NoMirror
      resourceName=None
      resourceObject=null
      tiling=NoTile
      imageSize=<0 0>
      margin=<ltrb 1 1 1 1>
      outlineSettings {
       bUseBrushTransparency=false
       roundingType=FixedRadius
       width=1
       color {colorUseRule=UseColor_ColorTable, specifiedColor=<rgba 0 0 0 0>}
       cornerRadii=<4 4 4 4>
      }
      tintColor {colorUseRule=UseColor_Specified, specifiedColor=<rgba 0.0395 0.0395 0.0395 1>}
      uVRegion {bIsValid=false, max=<0 0>, min=<0 0>}
     }
     uncheckedPressedImage {
      bIsDynamicallyLoaded=false
      drawAs=RoundedBox
      imageType=NoImage
      mirroring=NoMirror
      resourceName=None
      resourceObject=null
      tiling=NoTile
      imageSize=<0 0>
      margin=<ltrb 0 0 0 0>
      outlineSettings {
       bUseBrushTransparency=false
       roundingType=FixedRadius
       width=0
       color {colorUseRule=UseColor_Specified, specifiedColor=<rgba 0.0395 0.0395 0.0395 1>}
       cornerRadii=<4 4 4 4>
      }
      tintColor {colorUseRule=UseColor_Specified, specifiedColor=<rgba 0.0284 0.0284 0.0284 1>}
      uVRegion {bIsValid=false, max=<0 0>, min=<0 0>}
     }
     uncheckedSlateSound {resourceObject=null}
     undeterminedForeground {colorUseRule=UseColor_Specified, specifiedColor=<rgba 1 1 1 1>}
     undeterminedHoveredImage {
      bIsDynamicallyLoaded=false
      drawAs=Image
      imageType=Vector
      mirroring=NoMirror
      resourceName=../../../Engine/Content/Slate/Starship/CoreWidgets/CheckBox/indeterminate.svg
      resourceObject=null
      tiling=NoTile
      imageSize=<16 16>
      margin=<ltrb 0 0 0 0>
      outlineSettings {
       bUseBrushTransparency=false
       roundingType=HalfHeightRadius
       width=0
       color {colorUseRule=UseColor_Specified, specifiedColor=<rgba 0 0 0 0>}
       cornerRadii=<0 0 0 0>
      }
      tintColor {colorUseRule=UseColor_Specified, specifiedColor=<rgba 0.0044 0.2384 1 1>}
      uVRegion {bIsValid=false, max=<0 0>, min=<0 0>}
     }
     undeterminedImage {
      bIsDynamicallyLoaded=false
      drawAs=Image
      imageType=Vector
      mirroring=NoMirror
      resourceName=../../../Engine/Content/Slate/Starship/CoreWidgets/CheckBox/indeterminate.svg
      resourceObject=null
      tiling=NoTile
      imageSize=<16 16>
      margin=<ltrb 0 0 0 0>
      outlineSettings {
       bUseBrushTransparency=false
       roundingType=HalfHeightRadius
       width=0
       color {colorUseRule=UseColor_Specified, specifiedColor=<rgba 0 0 0 0>}
       cornerRadii=<0 0 0 0>
      }
      tintColor {colorUseRule=UseColor_Specified, specifiedColor=<rgba 0 0.162 0.7454 1>}
      uVRegion {bIsValid=false, max=<0 0>, min=<0 0>}
     }
     undeterminedPressedImage {
      bIsDynamicallyLoaded=false
      drawAs=Image
      imageType=Vector
      mirroring=NoMirror
      resourceName=../../../Engine/Content/Slate/Starship/CoreWidgets/CheckBox/indeterminate.svg
      resourceObject=null
      tiling=NoTile
      imageSize=<16 16>
      margin=<ltrb 0 0 0 0>
      outlineSettings {
       bUseBrushTransparency=false
       roundingType=HalfHeightRadius
       width=0
       color {colorUseRule=UseColor_Specified, specifiedColor=<rgba 0 0 0 0>}
       cornerRadii=<0 0 0 0>
      }
      tintColor {colorUseRule=UseColor_Specified, specifiedColor=<rgba 0 0.0802 0.3515 1>}
      uVRegion {bIsValid=false, max=<0 0>, min=<0 0>}
     }
    }
   }
   slot {class=HorizontalBoxSlot, properties {size {sizeRule=Fill, value=1}}}
  }
  UEFN_TextBlock_C_9 {
   widgetClassPath=&UEFN_TextBlock
   namedSlotHost=AutoFire
   widget=UEFN_TextBlock_C_9
   properties {
    text=Auto Fire
    font {
     bForceMonospaced=false
     bMaterialIsStencil=false
     fontMaterial=null
     fontObject=&Roboto
     letterSpacing=0
     monospacedWidth=1
     size=12
     skewAmount=0
     typefaceFontName=Regular
     outlineSettings {
      bApplyOutlineToDropShadows=false
      bMiteredCorners=false
      bSeparateFillAlpha=false
      outlineMaterial=null
      outlineSize=0
      outlineColor=<rgba 0 0 0 1>
     }
    }
   }
  }
  Spacer_13 {
   widgetClassPath=&Spacer
   parent=VerticalBox_116
   widget=Spacer_13
   properties {size=<89.4531 11.1193>}
   slot {class=VerticalBoxSlot}
  }
  SingleFire {
   widgetClassPath=&EditorUtilityButton
   parent=VerticalBox_116
   widget=SingleFire
   properties {toolTipText=Play MetaSound first. Fires a single round.}
   slot {class=VerticalBoxSlot}
  }
  UEFN_TextBlock_C_7 {
   widgetClassPath=&UEFN_TextBlock
   parent=SingleFire
   widget=UEFN_TextBlock_C_7
   properties {
    text=Single Fire
    font {
     bForceMonospaced=false
     bMaterialIsStencil=false
     fontMaterial=null
     fontObject=&Roboto
     letterSpacing=0
     monospacedWidth=1
     size=12
     skewAmount=0
     typefaceFontName=Regular
     outlineSettings {
      bApplyOutlineToDropShadows=false
      bMiteredCorners=false
      bSeparateFillAlpha=false
      outlineMaterial=null
      outlineSize=0
      outlineColor=<rgba 0 0 0 1>
     }
    }
   }
   slot {class=ButtonSlot}
  }
  Spacer_4 {
   widgetClassPath=&Spacer
   parent=VerticalBox_116
   widget=Spacer_4
   properties {size=<89.4531 11.1193>}
   slot {class=VerticalBoxSlot}
  }
  UEFN_TextBlock_C_6 {
   widgetClassPath=&UEFN_TextBlock
   parent=VerticalBox_116
   widget=UEFN_TextBlock_C_6
   properties {
    text=ADS
    font {
     bForceMonospaced=false
     bMaterialIsStencil=false
     fontMaterial=null
     fontObject=&Roboto
     letterSpacing=0
     monospacedWidth=1
     size=12
     skewAmount=0
     typefaceFontName=Regular
     outlineSettings {
      bApplyOutlineToDropShadows=false
      bMiteredCorners=false
      bSeparateFillAlpha=false
      outlineMaterial=null
      outlineSize=0
      outlineColor=<rgba 0 0 0 1>
     }
    }
   }
   slot {class=VerticalBoxSlot, properties {horizontalAlignment=HAlign_Center}}
  }
  ADSCheckBox {
   widgetClassPath=&EditorUtilityCheckBox
   bIsVariable=true
   parent=VerticalBox_116
   widget=ADSCheckBox
   slot {class=VerticalBoxSlot, properties {horizontalAlignment=HAlign_Center}}
  }
  Spacer_5 {
   widgetClassPath=&Spacer
   parent=VerticalBox_116
   widget=Spacer_5
   properties {size=<89.4531 11.1193>}
   slot {class=VerticalBoxSlot}
  }
  UEFN_TextBlock_C_93 {
   widgetClassPath=&UEFN_TextBlock
   parent=VerticalBox_116
   widget=UEFN_TextBlock_C_93
   properties {
    text=Burst
    font {
     bForceMonospaced=false
     bMaterialIsStencil=false
     fontMaterial=null
     fontObject=&Roboto
     letterSpacing=0
     monospacedWidth=1
     size=12
     skewAmount=0
     typefaceFontName=Regular
     outlineSettings {
      bApplyOutlineToDropShadows=false
      bMiteredCorners=false
      bSeparateFillAlpha=false
      outlineMaterial=null
      outlineSize=0
      outlineColor=<rgba 0 0 0 1>
     }
    }
   }
   slot {class=VerticalBoxSlot, properties {horizontalAlignment=HAlign_Center}}
  }
  IsBurstCheckBox {
   widgetClassPath=&EditorUtilityCheckBox
   bIsVariable=true
   parent=VerticalBox_116
   widget=IsBurstCheckBox
   properties {toolTipText=Enables burst behavior for Single Fire playback. Use Burst settings to update.}
   slot {class=VerticalBoxSlot, properties {horizontalAlignment=HAlign_Center}}
  }
  Spacer_6 {
   widgetClassPath=&Spacer
   parent=VerticalBox_116
   widget=Spacer_6
   properties {size=<89.4531 11.1193>}
   slot {class=VerticalBoxSlot}
  }
  UEFN_TextBlock_C {
   widgetClassPath=&UEFN_TextBlock
   parent=VerticalBox_116
   widget=UEFN_TextBlock_C
   properties {
    text=Use Ammo
    font {
     bForceMonospaced=false
     bMaterialIsStencil=false
     fontMaterial=null
     fontObject=&Roboto
     letterSpacing=0
     monospacedWidth=1
     size=12
     skewAmount=0
     typefaceFontName=Regular
     outlineSettings {
      bApplyOutlineToDropShadows=false
      bMiteredCorners=false
      bSeparateFillAlpha=false
      outlineMaterial=null
      outlineSize=0
      outlineColor=<rgba 0 0 0 1>
     }
    }
   }
   slot {class=VerticalBoxSlot, properties {horizontalAlignment=HAlign_Center}}
  }
  UseAmmoCheckBox {
   widgetClassPath=&EditorUtilityCheckBox
   bIsVariable=true
   parent=VerticalBox_116
   widget=UseAmmoCheckBox
   properties {checkedState=Checked}
   slot {class=VerticalBoxSlot, properties {horizontalAlignment=HAlign_Center}}
  }
  Spacer_12 {
   widgetClassPath=&Spacer
   parent=VerticalBox_116
   widget=Spacer_12
   properties {size=<89.4531 11.1193>}
   slot {class=VerticalBoxSlot}
  }
  Spacer_7 {
   widgetClassPath=&Spacer
   parent=VerticalBox_116
   widget=Spacer_7
   properties {size=<89.4531 11.1193>}
   slot {class=VerticalBoxSlot}
  }
  UEFN_TextBlock_C_1 {
   widgetClassPath=&UEFN_TextBlock
   parent=VerticalBox_116
   widget=UEFN_TextBlock_C_1
   properties {
    text=Fire Rate
    font {
     bForceMonospaced=false
     bMaterialIsStencil=false
     fontMaterial=null
     fontObject=&Roboto
     letterSpacing=0
     monospacedWidth=1
     size=12
     skewAmount=0
     typefaceFontName=Regular
     outlineSettings {
      bApplyOutlineToDropShadows=false
      bMiteredCorners=false
      bSeparateFillAlpha=false
      outlineMaterial=null
      outlineSize=0
      outlineColor=<rgba 0 0 0 1>
     }
    }
   }
   slot {class=VerticalBoxSlot, properties {horizontalAlignment=HAlign_Center}}
  }
  FireRateEntry {
   widgetClassPath=/TechAudioToolsContent/Widgets/MetaSoundLiterals/Float/Primitives/WBP_MetaSoundLiteralFloatPrimitive_Spinbox
   bIsVariable=true
   parent=VerticalBox_116
   widget=FireRateEntry
   properties {
    input Name=Weapon.FireRate
    metaSoundLiteralViewModel_Float=&self:WidgetTree.FireRateEntry.MetaSoundLiteralViewModel_Float_0
   }
   slot {class=VerticalBoxSlot}
  }
  Spacer_8 {
   widgetClassPath=&Spacer
   parent=VerticalBox_116
   widget=Spacer_8
   properties {size=<89.4531 11.1193>}
   slot {class=VerticalBoxSlot}
  }
  AmmoCountText {
   widgetClassPath=&UEFN_TextBlock
   bIsVariable=true
   parent=VerticalBox_116
   widget=AmmoCountText
   properties {
    text=Ammo Count
    font {
     bForceMonospaced=false
     bMaterialIsStencil=false
     fontMaterial=null
     fontObject=&Roboto
     letterSpacing=0
     monospacedWidth=1
     size=12
     skewAmount=0
     typefaceFontName=Regular
     outlineSettings {
      bApplyOutlineToDropShadows=false
      bMiteredCorners=false
      bSeparateFillAlpha=false
      outlineMaterial=null
      outlineSize=0
      outlineColor=<rgba 0 0 0 1>
     }
    }
   }
   slot {class=VerticalBoxSlot, properties {horizontalAlignment=HAlign_Center}}
  }
  AmmoCountEntry {
   widgetClassPath=/TechAudioToolsContent/Widgets/MetaSoundLiterals/Integer/Primitives/WBP_MetaSoundLiteralIntegerPrimitive_Spinbox
   bIsVariable=true
   parent=VerticalBox_116
   widget=AmmoCountEntry
   properties {
    input Name=Weapon.AmmoRemaining
    metaSoundLiteralViewModel_Integer=&self:WidgetTree.AmmoCountEntry.MetaSoundLiteralViewModel_Integer_0
   }
   slot {class=VerticalBoxSlot}
  }
  Spacer_9 {
   widgetClassPath=&Spacer
   parent=VerticalBox_116
   widget=Spacer_9
   properties {size=<89.4531 11.1193>}
   slot {class=VerticalBoxSlot}
  }
  WeaponWID {
   widgetClassPath=&SinglePropertyView
   parent=VerticalBox_116
   widget=WeaponWID
   properties {propertyName=WeaponItemDefinition, toolTipText=Set up WID to automatically parse fire rate and ammo count.}
   slot {class=VerticalBoxSlot}
  }
  Spacer_316 {widgetClassPath=&Spacer, parent=HorizontalBox_66, widget=Spacer_316, properties {size=<15.268 1>}, slot {class=HorizontalBoxSlot}}
  SettingsVerticalBox {
   widgetClassPath=&VerticalBox
   bIsVariable=true
   parent=HorizontalBox_66
   widget=SettingsVerticalBox
   slot {class=HorizontalBoxSlot}
  }
  Border_547 {
   widgetClassPath=&Border
   parent=SettingsVerticalBox
   widget=Border_547
   properties {
    background {
     bIsDynamicallyLoaded=false
     drawAs=RoundedBox
     imageType=NoImage
     mirroring=NoMirror
     resourceName=None
     resourceObject=null
     tiling=NoTile
     imageSize=<32 32>
     margin=<ltrb 0 0 0 0>
     outlineSettings {
      bUseBrushTransparency=false
      roundingType=FixedRadius
      width=0
      color {colorUseRule=UseColor_Specified, specifiedColor=<rgba 0 0 0 0>}
      cornerRadii=<12 12 12 12>
     }
     tintColor {colorUseRule=UseColor_Specified, specifiedColor=<rgba 0.0212 0.0212 0.0212 1>}
     uVRegion {bIsValid=false, max=<0 0>, min=<0 0>}
    }
    padding=<ltrb 8 8 8 8>
   }
   slot {class=VerticalBoxSlot}
  }
  VerticalBox_206 {
   widgetClassPath=&VerticalBox
   parent=Border_547
   widget=VerticalBox_206
   properties {toolTipText=Updates playback to use leaping wave players between shots.Requires playback restart.}
   slot {class=BorderSlot, properties {padding=<ltrb 8 8 8 8>}}
  }
  UEFN_TextBlock_C_5 {
   widgetClassPath=&UEFN_TextBlock
   parent=VerticalBox_206
   widget=UEFN_TextBlock_C_5
   properties {
    text=Single Voice
    font {
     bForceMonospaced=false
     bMaterialIsStencil=false
     fontMaterial=null
     fontObject=&Roboto
     letterSpacing=0
     monospacedWidth=1
     size=12
     skewAmount=0
     typefaceFontName=Regular
     outlineSettings {
      bApplyOutlineToDropShadows=false
      bMiteredCorners=false
      bSeparateFillAlpha=false
      outlineMaterial=null
      outlineSize=0
      outlineColor=<rgba 0 0 0 1>
     }
    }
   }
   slot {class=VerticalBoxSlot, properties {horizontalAlignment=HAlign_Center}}
  }
  SingleVoice {
   widgetClassPath=/TechAudioToolsContent/Widgets/MetaSoundLiterals/Bool/Primitives/WBP_MetaSoundLiteralBoolPrimitive_CheckBox
   bIsVariable=true
   parent=VerticalBox_206
   widget=SingleVoice
   properties {
    input Name=bUseSingleVoice
    metaSoundLiteralViewModel_Boolean=&self:WidgetTree.SingleVoice.MetaSoundLiteralViewModel_Boolean_0
   }
   slot {class=VerticalBoxSlot, properties {horizontalAlignment=HAlign_Center}}
  }
  Spacer_15 {
   widgetClassPath=&Spacer
   parent=SettingsVerticalBox
   widget=Spacer_15
   properties {size=<89.4531 11.1193>}
   slot {class=VerticalBoxSlot}
  }
  BurstSettings {
   widgetClassPath=&Border
   bIsVariable=true
   parent=SettingsVerticalBox
   widget=BurstSettings
   properties {
    background {
     bIsDynamicallyLoaded=false
     drawAs=RoundedBox
     imageType=NoImage
     mirroring=NoMirror
     resourceName=None
     resourceObject=null
     tiling=NoTile
     imageSize=<32 32>
     margin=<ltrb 0 0 0 0>
     outlineSettings {
      bUseBrushTransparency=false
      roundingType=FixedRadius
      width=0
      color {colorUseRule=UseColor_Specified, specifiedColor=<rgba 0 0 0 0>}
      cornerRadii=<12 12 12 12>
     }
     tintColor {colorUseRule=UseColor_Specified, specifiedColor=<rgba 0.0212 0.0212 0.0212 1>}
     uVRegion {bIsValid=false, max=<0 0>, min=<0 0>}
    }
    padding=<ltrb 8 8 8 8>
   }
   slot {class=VerticalBoxSlot}
  }
  BurstSettingsBox {
   widgetClassPath=&VerticalBox
   parent=BurstSettings
   widget=BurstSettingsBox
   slot {class=BorderSlot, properties {padding=<ltrb 8 8 8 8>}}
  }
  Spacer_10 {
   widgetClassPath=&Spacer
   parent=BurstSettingsBox
   widget=Spacer_10
   properties {size=<89.4531 11.1193>}
   slot {class=VerticalBoxSlot}
  }
  UEFN_TextBlock_C_3 {
   widgetClassPath=&UEFN_TextBlock
   parent=BurstSettingsBox
   widget=UEFN_TextBlock_C_3
   properties {
    text=Burst Rate
    font {
     bForceMonospaced=false
     bMaterialIsStencil=false
     fontMaterial=null
     fontObject=&Roboto
     letterSpacing=0
     monospacedWidth=1
     size=12
     skewAmount=0
     typefaceFontName=Regular
     outlineSettings {
      bApplyOutlineToDropShadows=false
      bMiteredCorners=false
      bSeparateFillAlpha=false
      outlineMaterial=null
      outlineSize=0
      outlineColor=<rgba 0 0 0 1>
     }
    }
   }
   slot {class=VerticalBoxSlot, properties {horizontalAlignment=HAlign_Center}}
  }
  BurstFireRateEntry {
   widgetClassPath=&WBP_MetaSoundLiteralFloatPrimitive_Spinbox
   bIsVariable=true
   parent=BurstSettingsBox
   widget=BurstFireRateEntry
   properties {
    input Name=BurstRate
    metaSoundLiteralViewModel_Float=&self:WidgetTree.BurstFireRateEntry.MetaSoundLiteralViewModel_Float_0
   }
   slot {class=VerticalBoxSlot}
  }
  Spacer_11 {
   widgetClassPath=&Spacer
   parent=BurstSettingsBox
   widget=Spacer_11
   properties {size=<89.4531 11.1193>}
   slot {class=VerticalBoxSlot}
  }
  UEFN_TextBlock_C_4 {
   widgetClassPath=&UEFN_TextBlock
   parent=BurstSettingsBox
   widget=UEFN_TextBlock_C_4
   properties {
    text=Burst Count
    font {
     bForceMonospaced=false
     bMaterialIsStencil=false
     fontMaterial=null
     fontObject=&Roboto
     letterSpacing=0
     monospacedWidth=1
     size=12
     skewAmount=0
     typefaceFontName=Regular
     outlineSettings {
      bApplyOutlineToDropShadows=false
      bMiteredCorners=false
      bSeparateFillAlpha=false
      outlineMaterial=null
      outlineSize=0
      outlineColor=<rgba 0 0 0 1>
     }
    }
   }
   slot {class=VerticalBoxSlot, properties {horizontalAlignment=HAlign_Center}}
  }
  BurstCountEntry {
   widgetClassPath=&WBP_MetaSoundLiteralIntegerPrimitive_Spinbox
   bIsVariable=true
   parent=BurstSettingsBox
   widget=BurstCountEntry
   properties {
    input Name=RoundsPerBurst
    metaSoundLiteralViewModel_Integer=&self:WidgetTree.BurstCountEntry.MetaSoundLiteralViewModel_Integer_0
   }
   slot {class=VerticalBoxSlot}
  }
 ]
}
