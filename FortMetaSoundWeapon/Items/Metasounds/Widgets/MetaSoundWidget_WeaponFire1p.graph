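# Editor utility MetaSound preset widget for first-person weapon fire MetaSounds, with a Fire button and a Press and Hold checkbox.
# With hold enabled, press and release set firing on and off; otherwise each click toggles it through a FlipFlop.
# A Set Is Firing event writes the boolean to the bIsFiring graph input default on the Builder cached at construction.
# It applies only to MetaSounds listed in Supported MetaSounds, empty by default.
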EditorUtilityWidgetBlueprint /FortMetaSoundWeapon/Items/Metasounds/Widgets/MetaSoundWidget_WeaponFire1p
graphs [
 EventGraph {nodes [
   n1 = "On Check State Changed (EditorUtilityCheckBox_105)"!() -> (OutputDelegate, then, bIsChecked:bool) {is=title K2Node_ComponentBoundEvent}
   > Set Enable Press and Hold(Enable Press and Hold:bool=@n1.bIsChecked) -> (Output_Get:bool)
   "On Clicked (FireButton)"() {is=title K2Node_ComponentBoundEvent}
   > Branch(Condition=NOT Boolean(A=Get Enable Press and Hold() -> (Enable Press and Hold:bool)))
   > n2 = Flip Flop!(None:Exec) -> (A:Exec, B:Exec, IsA:bool=~) {is=title K2Node_MacroInstance}
   Set Is Firing!(execute=[@n2.A, @n2.B], self, bIsFiring:bool=@n2) -> (then) {is=title K2Node_CallFunction}
   "On Pressed (FireButton)"() {is=title K2Node_ComponentBoundEvent}
   > Branch(Condition=Get Enable Press and Hold())
   > Set Is Firing(bIsFiring=true)
   "On Released (FireButton)"() {is=title K2Node_ComponentBoundEvent}
   > Branch(Condition=Get Enable Press and Hold())
   > Set Is Firing(bIsFiring=false)
   n3 = Set Is Firing!() -> (OutputDelegate, then, bIsFiring:bool=~) {is=title K2Node_CustomEvent}
   > n4 = Create MetaSound Literal from Boolean!(execute, Value:bool=@n3.bIsFiring) -> (then, ReturnValue:Meta Sound Literal Structure)
   > SetGraphInputDefault!(
    execute
    self:Meta Sound Builder Base Object Reference=Get Builder() -> (Builder:Meta Sound Builder Base Object Reference)
    InputName:Name=bIsFiring
    Literal:"Meta Sound Literal Structure (by ref)"=@n4
    PageName:Name
   ) -> (Succeeded:Exec, Failed:Exec)
   n5 = Event On MetaSoundPreset Widget Constructed!() -> (OutputDelegate, then, Builder:Meta Sound Builder Base Object Reference) {is=title K2Node_Event}
   > Set Builder(Builder:Meta Sound Builder Base Object Reference=@n5.Builder) -> (Output_Get:Meta Sound Builder Base Object Reference)
  ]}
 GetEditorName {nodes [
   GetEditorName() {is=title K2Node_FunctionEntry}
   > Return Node!(execute, ReturnValue:Text)
  ]}
 GetIconBrushName {nodes [
   GetIconBrushName() {is=title K2Node_FunctionEntry}
   > Return Node(ReturnValue:Name=None)
  ]}
 GetSupportedMetaSounds {nodes [
   GetSupportedMetaSounds() {is=title K2Node_FunctionEntry}
   > Return Node!(
    execute
    bSupportAllPresets:bool=false
    ExcludedMetaSounds:Array of Meta Sound Document Interface Interfaces
    IncludedMetaSounds:Array of Meta Sound Document Interface Interfaces=Get Supported MetaSounds() -> (Supported MetaSounds:Array of Meta Sound Document Interface Interfaces)
   )
  ]}
 OnInitializingTab {nodes [
   On Initializing Tab() {is=title K2Node_FunctionEntry}
   > Return Node!(execute, TabInfo:Meta Sound Template Widget Tab Info Structure)
  ]}
]
variables [
 Builder {type=MetaSoundBuilderBase, default=null}
 Enable Press and Hold {type=boolean, default=true}
 Supported MetaSounds {type="array<object>", default []}
]
widgets {
 info {rootWidgetClass=&VerticalBox, widgetCount=6}
 widgets [
  VerticalBox_22 {widgetClassPath=&VerticalBox, widget=VerticalBox_22}
  FireButton {
   widgetClassPath=&EditorUtilityButton
   parent=VerticalBox_22
   widget=FireButton
   slot {class=VerticalBoxSlot, properties {horizontalAlignment=HAlign_Left, verticalAlignment=VAlign_Center}}
  }
  TextBlock_27 {
   widgetClassPath=&TextBlock
   parent=FireButton
   widget=TextBlock_27
   properties {text=Fire}
   slot {class=ButtonSlot, properties {horizontalAlignment=HAlign_Fill, verticalAlignment=VAlign_Fill, padding=<ltrb 0 0 0 0>}}
  }
  HorizontalBox_0 {widgetClassPath=&HorizontalBox, parent=VerticalBox_22, widget=HorizontalBox_0, slot {class=VerticalBoxSlot}}
  EditorUtilityCheckBox_105 {
   widgetClassPath=&EditorUtilityCheckBox
   bIsVariable=true
   parent=HorizontalBox_0
   widget=EditorUtilityCheckBox_105
   properties {checkedState=Checked}
   slot {class=HorizontalBoxSlot}
  }
  TextBlock_141 {
   widgetClassPath=&TextBlock
   parent=HorizontalBox_0
   widget=TextBlock_141
   properties {
    text=Press and Hold
    font {
     bForceMonospaced=false
     bMaterialIsStencil=false
     fontMaterial=null
     fontObject=&Roboto
     letterSpacing=0
     monospacedWidth=1
     size=9
     skewAmount=0
     typefaceFontName=Regular
     outlineSettings {
      bApplyOutlineToDropShadows=false
      bMiteredCorners=false
      bSeparateFillAlpha=false
      outlineMaterial=null
      outlineSize=0
      outlineColor=<rgba 0 0 0 1>
     }
    }
   }
   slot {class=HorizontalBoxSlot, properties {verticalAlignment=VAlign_Center}}
  }
 ]
}
