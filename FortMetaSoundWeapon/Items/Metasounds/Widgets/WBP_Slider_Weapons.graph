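# WBP_Slider_Weapons is a vertical audio slider widget that edits a float MetaSound input, chosen by Input Name from the supplied viewmodels.
# Slider changes map into the literal viewmodel's source range, linearly or logarithmically, set SourceValue, then fire OnSliderValueChanged.
# An Is Constructing flag, cleared one tick after Construct, suppresses startup changes.
# PreConstruct applies Slider Style, and mouse enter/leave broadcasts the Is Hovered field.
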WidgetBlueprint /FortMetaSoundWeapon/Items/Metasounds/Widgets/WBP_Slider_Weapons
graphs [
 EventGraph {nodes [
   n1 = "On Value Changed (AudioMaterialSlider)"!() -> (OutputDelegate, then, Value:float) {is=title K2Node_ComponentBoundEvent}
   > n2 = Get!(execute:Exec) -> (then:Exec, else:Exec, Is Constructing:bool)
   n3 = Get MetaSoundLiteralViewModel_Float() -> (MetaSoundLiteralViewModel_Float:Meta Sound Literal Float Viewmodel Object Reference)
   n4 = GetSourceRangeMax!(self:Meta Sound Literal Float Viewmodel Object Reference=@n3) -> (ReturnValue:float)
   n5 = GetSourceRangeMin!(self:Meta Sound Literal Float Viewmodel Object Reference=@n3) -> (ReturnValue:float)
   n6 = Lerp(A=Loge!(A:double=@n5) -> (ReturnValue:double), B=Loge(A=@n4), Alpha=@n1.Value)
   n7 = Set SourceValue!(
    execute=@n2.else
    SourceValue:float=SelectFloat(
     A=Exp!(A:double=@n6) -> (ReturnValue:double)
     B=Lerp(A=@n5, B=@n4, Alpha=@n1.Value)
     bPickA=Get Apply Logarithmic Interpolation() -> (Apply Logarithmic Interpolation:bool)
    )
    self:Meta Sound Literal Float Viewmodel Object Reference=Get MetaSoundLiteralViewModel_Float()
   ) -> (then, Output_Get:float)
   > Call On Slider Value Changed!(execute:Exec, self:WBP Slider Weapons Object Reference, Value:double=@n7) -> (then:Exec) {is=title K2Node_CallDelegate}
   Event Construct()
   > Set Is Constructing(Is Constructing:bool=true) -> (Output_Get:bool)
   > DelayUntilNextTick!(execute) -> (then)
   > Set Is Constructing(Is Constructing=false)
   n8 = Event OnMouseEnter()
   Broadcast Field Value Changed!(execute=[@n8.then, @n9.then], self:Widget Object Reference, FieldId:Field Notification Id Structure=(FieldName="Is Hovered")) -> (then)
   n9 = Event OnMouseLeave!() -> (OutputDelegate, then, MouseEvent:"Pointer Event Structure (by ref)") {is=title K2Node_Event}
   Event PreConstruct()
   > Set WidgetStyle!(
    execute
    WidgetStyle:Audio Material Slider Style Structure=Get Slider Style() -> (Slider Style:Audio Material Slider Style Structure)
    self:Audio Material Slider Object Reference=Get AudioMaterialSlider() -> (AudioMaterialSlider:Audio Material Slider Object Reference)
   ) -> (then, Output_Get:Audio Material Slider Style Structure)
   n10 = Event Set Input Viewmodels!() -> (OutputDelegate, then, InputViewModels:Map of Names to Meta Sound Input Viewmodel Object References) {is=title K2Node_Event}
   > SetMetaSoundInputViewModel!(
    execute
    self
    Viewmodel:Meta Sound Input Viewmodel Object Reference=Find(
     TargetMap:Map of Names to Meta Sound Input Viewmodel Object References=@n10.InputViewModels
     Key:"Name (by ref)"=Get Input Name() -> (Input Name:Name=None)
    ) -> (Value:Meta Sound Input Viewmodel Object Reference)
   ) -> (then)
  ]}
 GetInputViewModelNames {nodes [
   Get Input Viewmodel Names() {is=title K2Node_FunctionEntry}
   > Return Node!(execute, ReturnValue:Array of Names=Make Array("[0]":Name=Get Input Name() -> (Input Name=None)) -> (Array:Array of Names))
  ]}
 Is Hovered {nodes [
   Is Hovered() {is=title K2Node_FunctionEntry}
   > Return Node(ReturnValue:bool=IsHovered())
  ]}
 OnSliderValueChanged {nodes [OnSliderValueChanged!() -> (then, Value:double=~) {is=title K2Node_FunctionEntry}]}
]
variables [
 Apply Logarithmic Interpolation {type=boolean}
 Input Name {type=string, default=None}
 Is Constructing {type=boolean}
 OnSliderValueChanged {}
 Slider Style {
  type=AudioMaterialSliderStyle
  default {
   material=null
   desiredSize=<25 250>
   sliderBackgroundAccentColor=<rgba 0.005 0.005 0.005 1>
   sliderBackgroundColor=<rgba 0.008 0.008 0.008 1>
   sliderHandleMainColor=<rgba 1 1 1 1>
   sliderHandleOutlineColor=<rgba 0.15 0.15 0.15 1>
   sliderValueMainColor=<rgba 0.008 0.008 0.008 1>
   textBoxStyle {
    backgroundColor {colorUseRule=UseColor_ColorTable, specifiedColor=<rgba 0 0 0 0>}
    backgroundImage {
     bIsDynamicallyLoaded=false
     drawAs=RoundedBox
     imageType=NoImage
     mirroring=NoMirror
     resourceName=None
     resourceObject=null
     tiling=NoTile
     imageSize=<56 28>
     margin=<ltrb 0 0 0 0>
     outlineSettings {
      bUseBrushTransparency=false
      roundingType=FixedRadius
      width=0
      color {colorUseRule=UseColor_Specified, specifiedColor=<rgba 0 0 0 0>}
      cornerRadii=<4 4 4 4>
     }
     tintColor {colorUseRule=UseColor_ColorTable, specifiedColor=<rgba 0 0 0 0>}
     uVRegion {bIsValid=false, max=<0 0>, min=<0 0>}
    }
   }
  }
 }
]
widgets {
 info {inheritedWidgetCount=0, parentClass=&UserWidget, rootWidgetClass=/Script/AudioWidgets.AudioMaterialSlider, widgetCount=1}
 widgets [AudioMaterialSlider {
   widgetClassPath=&AudioMaterialSlider
   bIsVariable=true
   widget=AudioMaterialSlider
   properties {orientation=Orient_Vertical}
  }]
}
